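# Control Rig for the female Ceremonial Guard Fencer body that applies pose-driven corrective offsets to costume dyn_ bones.
# A sequence runs spherical pose readers that weight Modify Transforms: left thigh and spine drive jacket bones, right thigh drives belt and badge bones.
# Both clavicle readers multiply to adjust the shoulder belt, and a curve reshapes the front jacket weight.
# Additional arm, clavicle and jacket corrective chains are disconnected from execution.
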ControlRigBlueprint /BRCosmetics/Characters/Player/Female/Medium/Bodies/F_MED_CeremonialGuard_Fencer/Meshes/F_MED_CeremonialGuard_Fencer_CtrlRig
graphs [
 RigVMFunctionLibrary {links [], nodes []}
 RigVMFunction_Sequence_1_1_ContainedGraph {nodes [
   n1 = RigVMFunctionEntryNode()
   n2 = &RigVMFunction_Sequence(ExecuteContext=@n1.ExecuteContext)
   n3 = &RigVMFunction_Sequence(ExecuteContext=@n2.B)
   RigVMFunctionReturnNode(A=@n2.A, B=@n3.A, C=@n3.B)
  ]}
 RigVMModel {nodes [
   n1 = &RigUnit_SphericalPoseReader(
    ActiveRegionScaleFactors=(PositiveWidth=4,NegativeWidth=0.2,PositiveHeight=3,NegativeHeight=4)
    ActiveRegionSize=0.1
    Debug=(bDrawDebug=false,bDraw2D=false,bDrawLocalAxes=false,DebugScale=30,DebugSegments=30,DebugThickness=0.2)
    DriverAxis=(X=-1,Y=0,Z=0)
    DriverItem=(Type=Bone,Name="thigh_l")
    FalloffRegionScaleFactors=(PositiveWidth=1,NegativeWidth=1.1,PositiveHeight=1,NegativeHeight=1)
    FalloffSize=0.4
    FlipWidthScaling=false
    OptionalParentItem=(Type=Bone,Name="pelvis")
    RotationOffset=(X=110,Y=15,Z=0)
    ExecutePin=@RigVMFunction_Sequence_1.A
   )
   RigVMFunction_AnimEvalRichCurve = &RigVMFunction_AnimEvalRichCurve(
    Curve=(EditorCurveData=(Keys=((InterpMode=RCIM_Cubic,TangentMode=RCTM_Break,Time=0.4072,Value=0.0127,LeaveTangent=0.4607),(InterpMode=RCIM_Cubic,TangentMode=RCTM_Break,Time=0.9614,Value=0.9849,ArriveTangent=2.1545))))
    SourceMinimum=0
    TargetMinimum=0
    Value=@n1.OutputParam
   )
   &RigUnit_ModifyTransforms(
    ExecutePin=()
    ItemToModify=((Item=(Type=Bone,Name="dyn_BACE_jacet_front"),Transform=(Rotation=(X=0,Y=0,Z=0,W=1),Translation=(X=0,Y=0,Z=0),Scale3D=(X=1,Y=1,Z=1))))
    WeightMinimum=0
    Weight=@RigVMFunction_AnimEvalRichCurve.Result
    ItemToModify.0.Transform=/Script/RigVM.RigVMFunction_MathTransformFromSRT(Location=(X=0,Y=0,Z=0), Rotation=(X=0,Y=0,Z=20), RotationOrder=XYZ, Scale=(X=1,Y=1,Z=1)) {is=struct RigVMUnitNode}.Transform
   )
   > &RigUnit_ModifyTransforms(
    ItemToModify=((Item=(Type=Bone,Name="dyn_BACE_jacet_front_mid"),Transform=(Rotation=(X=0,Y=0,Z=0,W=1),Translation=(X=0,Y=0,Z=0),Scale3D=(X=1,Y=1,Z=1))))
    WeightMinimum=0
    Weight=@RigVMFunction_AnimEvalRichCurve.Result
    ItemToModify.0.Transform=&RigVMFunction_MathTransformFromSRT(Location=(X=0,Y=0,Z=0), Rotation=(X=0,Y=10,Z=10), RotationOrder=XYZ, Scale=(X=1,Y=1,Z=1)).Transform
   )
   RigUnit_BeginExecution = &RigUnit_BeginExecution()
   RigVMFunction_Sequence_1 = RigVMAggregateNode(ExecuteContext=@RigUnit_BeginExecution.ExecutePin)
   n2 = &RigUnit_SphericalPoseReader(
    ActiveRegionScaleFactors=(PositiveWidth=1,NegativeWidth=0.4,PositiveHeight=1,NegativeHeight=1)
    ActiveRegionSize=0.3
    Debug=(bDrawDebug=false,bDraw2D=false,bDrawLocalAxes=false,DebugScale=30,DebugSegments=30,DebugThickness=0.2)
    DriverAxis=(X=1,Y=0,Z=0)
    DriverItem=(Type=Bone,Name="thigh_r")
    FalloffRegionScaleFactors=(PositiveWidth=1,NegativeWidth=1.6,PositiveHeight=1,NegativeHeight=1)
    FalloffSize=0.4
    OptionalParentItem=(Type=Bone,Name="pelvis")
    RotationOffset=(X=-110,Y=180,Z=0)
    ExecutePin=@RigVMFunction_Sequence_1.B
   )
   > &RigUnit_ModifyTransforms(
    ItemToModify=((Item=(Type=Bone,Name="dyn_beld1"),Transform=(Rotation=(X=0,Y=0,Z=0,W=1),Translation=(X=0,Y=0,Z=0),Scale3D=(X=1,Y=1,Z=1))))
    Weight=@n2.OutputParam
    ItemToModify.0.Transform=&RigVMFunction_MathTransformFromSRT(Location=(X=0,Y=-0.5,Z=1.5), Rotation=(X=-5,Y=12,Z=0), Scale=(X=1,Y=1,Z=1)).Transform
   )
   > &RigUnit_ModifyTransforms(
    ItemToModify=((Item=(Type=Bone,Name="dyn_BACE_scale_trans"),Transform=(Rotation=(X=0,Y=0,Z=0,W=1),Translation=(X=0,Y=0,Z=0),Scale3D=(X=1,Y=1,Z=1))))
    Weight=@n2.OutputParam
    ItemToModify.0.Transform=&RigVMFunction_MathTransformFromSRT(Location=(X=-2,Y=0,Z=-1), Rotation=(X=-2,Y=0,Z=0), Scale=(X=1,Y=1,Z=1)).Transform
   )
   > &RigUnit_ModifyTransforms(
    ItemToModify=((Item=(Type=Bone,Name="dyn_badge_down"),Transform=(Rotation=(X=0,Y=0,Z=0,W=1),Translation=(X=0,Y=0,Z=0),Scale3D=(X=1,Y=1,Z=1))))
    WeightMinimum=0
    ItemToModify.0.Transform=&RigVMFunction_MathTransformFromSRT(Location=(X=1,Y=0,Z=0), Rotation=(X=0,Y=0,Z=0), RotationOrder=XYZ, Scale=(X=1,Y=1,Z=1)).Transform
    Weight=@n2.OutputParam
   )
   n3 = &RigUnit_SphericalPoseReader(
    ActiveRegionScaleFactors=(PositiveWidth=1,NegativeWidth=1,PositiveHeight=0.5,NegativeHeight=0.5)
    ActiveRegionSize=0.1
    Debug=(bDrawDebug=false,bDraw2D=false,bDrawLocalAxes=false,DebugScale=30,DebugSegments=30,DebugThickness=0.2)
    DriverAxis=(X=-1,Y=0,Z=0)
    DriverItem=(Type=Bone,Name="clavicle_r")
    FalloffRegionScaleFactors=(PositiveWidth=1,NegativeWidth=0.55,PositiveHeight=1,NegativeHeight=1)
    FalloffSize=0.4
    FlipWidthScaling=false
    OptionalParentItem=(Type=Bone,Name="pelvis")
    RotationOffset=(X=180,Y=-40,Z=0)
    ExecutePin=@RigVMFunction_Sequence_1.C
   )
   > n4 = &RigUnit_SphericalPoseReader(
    ActiveRegionScaleFactors=(PositiveWidth=1,NegativeWidth=1,PositiveHeight=0.5,NegativeHeight=0.5)
    ActiveRegionSize=0.1
    Debug=(bDrawDebug=true,bDraw2D=false,bDrawLocalAxes=false,DebugScale=30,DebugSegments=30,DebugThickness=0.2)
    DriverAxis=(X=1,Y=0,Z=0)
    DriverItem=(Type=Bone,Name="clavicle_l")
    FalloffRegionScaleFactors=(PositiveWidth=1,NegativeWidth=0.55,PositiveHeight=1,NegativeHeight=1)
    FalloffSize=0.4
    FlipWidthScaling=false
    OptionalParentItem=(Type=Bone,Name="pelvis")
    RotationOffset=(X=0,Y=-40,Z=0)
   )
   > &RigUnit_ModifyTransforms(
    ItemToModify=((Item=(Type=Bone,Name="dyn_Belt_sholders"),Transform=(Rotation=(X=0,Y=0,Z=0,W=1),Translation=(X=0,Y=0,Z=0),Scale3D=(X=1,Y=1,Z=1))))
    WeightMinimum=0
    Weight=&RigVMFunction_MathDoubleMul(B=@n3.OutputParam, A=@n4.OutputParam).Result
    ItemToModify.0.Transform=&RigVMFunction_MathTransformFromSRT(Location=(X=4,Y=1.5,Z=0), Rotation=(X=0,Y=0,Z=0), RotationOrder=XYZ, Scale=(X=0.9,Y=1,Z=1)).Transform
   )
   n5 = &RigUnit_SphericalPoseReader(
    ActiveRegionScaleFactors=(PositiveWidth=0.8,NegativeWidth=0.1,PositiveHeight=0.2,NegativeHeight=0.2)
    ActiveRegionSize=0.2
    Debug=(bDrawDebug=false,bDraw2D=false,bDrawLocalAxes=false,DebugScale=30,DebugSegments=30,DebugThickness=0.2)
    DriverAxis=(X=-1,Y=0,Z=0)
    DriverItem=(Type=Bone,Name="clavicle_r")
    ExecutePin=()
    FalloffRegionScaleFactors=(PositiveWidth=1,NegativeWidth=0.2,PositiveHeight=0.25,NegativeHeight=1)
    FalloffSize=0.3
    FlipWidthScaling=false
    OptionalParentItem=(Type=Bone,Name="lowerarm_r")
    RotationOffset=(X=140,Y=-68,Z=0)
   )
   > &RigUnit_ModifyTransforms(
    ItemToModify=((Item=(Type=Bone,Name="dyn_Belt_start"),Transform=(Rotation=(X=0,Y=0,Z=0,W=1),Translation=(X=0,Y=0,Z=0),Scale3D=(X=1,Y=1,Z=1))))
    WeightMinimum=0
    ItemToModify.0.Transform=&RigVMFunction_MathTransformFromSRT(Location=(X=0.5,Y=0,Z=0), Rotation=(X=-10,Y=7,Z=5), RotationOrder=XYZ, Scale=(X=1,Y=1,Z=1)).Transform
    Weight=@n5.OutputParam
   )
   > &RigUnit_ModifyTransforms(
    ItemToModify=((Item=(Type=Bone,Name="dyn_Belt_sholders"),Transform=(Rotation=(X=0,Y=0,Z=0,W=1),Translation=(X=0,Y=0,Z=0),Scale3D=(X=1,Y=1,Z=1))))
    WeightMinimum=0
    ItemToModify.0.Transform=&RigVMFunction_MathTransformFromSRT(Location=(X=0,Y=1,Z=-1), Rotation=(X=-15,Y=10,Z=0), RotationOrder=XYZ, Scale=(X=0.7,Y=1,Z=1)).Transform
    Weight=@n5.OutputParam
   )
   > n6 = &RigUnit_SphericalPoseReader(
    ActiveRegionScaleFactors=(PositiveWidth=1,NegativeWidth=1,PositiveHeight=0.1,NegativeHeight=0.1)
    ActiveRegionSize=0.5
    Debug=(bDrawDebug=true,bDraw2D=false,bDrawLocalAxes=false,DebugScale=30,DebugSegments=30,DebugThickness=0.2)
    DriverAxis=(X=0.5,Y=0.1,Z=0.1)
    DriverItem=(Type=Bone,Name="upperarm_l")
    FalloffRegionScaleFactors=(PositiveWidth=2,NegativeWidth=2,PositiveHeight=0.2,NegativeHeight=0.2)
    FalloffSize=0.3
    FlipWidthScaling=false
    OptionalParentItem=(Type=Bone,Name="pelvis")
    RotationOffset=(X=83,Y=-50,Z=0)
   )
   > &RigUnit_ModifyTransforms(
    ItemToModify=((Item=(Type=Bone,Name="dyn_Belt_start"),Transform=(Rotation=(X=0,Y=0,Z=0,W=1),Translation=(X=0,Y=0,Z=0),Scale3D=(X=1,Y=1,Z=1))))
    WeightMinimum=0
    ItemToModify.0.Transform=&RigVMFunction_MathTransformFromSRT(Location=(X=0,Y=0,Z=0), Rotation=(X=0,Y=-5,Z=4), RotationOrder=XYZ, Scale=(X=1,Y=1,Z=1)).Transform
    Weight=@n6.OutputParam
   )
   n7 = &RigUnit_SphericalPoseReader(
    ActiveRegionScaleFactors=(PositiveWidth=1,NegativeWidth=1,PositiveHeight=1,NegativeHeight=1)
    ActiveRegionSize=0.02
    Debug=(bDrawDebug=true,bDraw2D=false,bDrawLocalAxes=false,DebugScale=30,DebugSegments=30,DebugThickness=0.2)
    DriverAxis=(X=1,Y=0.2,Z=0.1)
    DriverItem=(Type=Bone,Name="upperarm_l")
    ExecutePin=()
    FalloffRegionScaleFactors=(PositiveWidth=1,NegativeWidth=1,PositiveHeight=1,NegativeHeight=1)
    FalloffSize=0.3
    FlipWidthScaling=false
    OptionalParentItem=(Type=Bone,Name="pelvis")
    RotationOffset=(X=0,Y=-140,Z=5)
   )
   > &RigUnit_ModifyTransforms(
    ItemToModify=((Item=(Type=Bone,Name="dyn_Belt_start"),Transform=(Rotation=(X=0,Y=0,Z=0,W=1),Translation=(X=0,Y=0,Z=0),Scale3D=(X=1,Y=1,Z=1))))
    WeightMinimum=0
    ItemToModify.0.Transform=&RigVMFunction_MathTransformFromSRT(Location=(X=0,Y=0,Z=0), Rotation=(X=0,Y=-7,Z=0), RotationOrder=XYZ, Scale=(X=1,Y=1,Z=1)).Transform
    Weight=@n7.OutputParam
   )
   > n8 = &RigUnit_SphericalPoseReader(
    ActiveRegionScaleFactors=(PositiveWidth=1,NegativeWidth=1,PositiveHeight=1,NegativeHeight=1)
    ActiveRegionSize=0.02
    Debug=(bDrawDebug=true,bDraw2D=false,bDrawLocalAxes=false,DebugScale=30,DebugSegments=30,DebugThickness=0.2)
    DriverAxis=(X=1,Y=0.2,Z=0.1)
    DriverItem=(Type=Bone,Name="upperarm_l")
    FalloffRegionScaleFactors=(PositiveWidth=1,NegativeWidth=1,PositiveHeight=1,NegativeHeight=1)
    FalloffSize=0.3
    FlipWidthScaling=false
    OptionalParentItem=(Type=Bone,Name="pelvis")
    RotationOffset=(X=44,Y=40,Z=5)
   )
   > &RigUnit_ModifyTransforms(
    ItemToModify=((Item=(Type=Bone,Name="dyn_Belt_start"),Transform=(Rotation=(X=0,Y=0,Z=0,W=1),Translation=(X=0,Y=0,Z=0),Scale3D=(X=1,Y=1,Z=1))))
    WeightMinimum=0
    ItemToModify.0.Transform=&RigVMFunction_MathTransformFromSRT(Location=(X=0,Y=0,Z=0), Rotation=(X=0,Y=-5,Z=4), RotationOrder=XYZ, Scale=(X=1,Y=1,Z=1)).Transform
    Weight=@n8.OutputParam
   )
   n9 = &RigUnit_SphericalPoseReader(
    ActiveRegionScaleFactors=(PositiveWidth=1,NegativeWidth=1,PositiveHeight=1,NegativeHeight=1)
    ActiveRegionSize=0.01
    Debug=(bDrawDebug=false,bDraw2D=false,bDrawLocalAxes=false,DebugScale=30,DebugSegments=30,DebugThickness=0.2)
    DriverAxis=(X=0,Y=1,Z=0)
    DriverItem=(Type=Bone,Name="spine_01")
    FalloffRegionScaleFactors=(PositiveWidth=0.5,NegativeWidth=1,PositiveHeight=1,NegativeHeight=1)
    FalloffSize=0.1
    FlipWidthScaling=false
    OptionalParentItem=(Type=Bone,Name="pelvis")
    RotationOffset=(X=88,Y=0,Z=22)
    ExecutePin=@n10.ExecutePin
   )
   > &RigUnit_ModifyTransforms(
    ItemToModify=((Item=(Type=Bone,Name="dyn_BACE_jacet_front"),Transform=(Rotation=(X=0,Y=0,Z=0,W=1),Translation=(X=0,Y=0,Z=0),Scale3D=(X=1,Y=1,Z=1))))
    WeightMinimum=0
    ItemToModify.0.Transform=&RigVMFunction_MathTransformFromSRT(Location=(X=0,Y=0,Z=0), Rotation=(X=0,Y=0,Z=15), RotationOrder=XYZ, Scale=(X=1,Y=1,Z=1)).Transform
    Weight=@n9.OutputParam
   )
   &RigUnit_ModifyTransforms(
    ItemToModify=((Item=(Type=Bone,Name="dyn_BACE_jacet_front"),Transform=(Rotation=(X=0,Y=0,Z=0,W=1),Translation=(X=0,Y=0,Z=0),Scale3D=(X=1,Y=1,Z=1))))
    WeightMinimum=0
    Weight=@RigVMFunction_AnimEvalRichCurve.Result
    ItemToModify.0.Transform=&RigVMFunction_MathTransformFromSRT(Location=(X=-1,Y=0,Z=0), Rotation=(X=15,Y=-20,Z=110), RotationOrder=XYZ, Scale=(X=1,Y=1,Z=1)).Transform
    ExecutePin=@n1.ExecutePin
   )
   > &RigUnit_ModifyTransforms(
    ItemToModify=((Item=(Type=Bone,Name="dyn_BACE_jacet_front_mid"),Transform=(Rotation=(X=0,Y=0,Z=0,W=1),Translation=(X=0,Y=0,Z=0),Scale3D=(X=1,Y=1,Z=1))))
    WeightMinimum=0
    ItemToModify.0.Transform=&RigVMFunction_MathTransformFromSRT(Location=(X=-1,Y=0,Z=0), Rotation=(X=0,Y=0,Z=45), RotationOrder=XYZ, Scale=(X=1,Y=1,Z=1)).Transform
    Weight=@n1.OutputParam
   )
   > n10 = &RigUnit_ModifyTransforms(
    ItemToModify=((Item=(Type=Bone,Name="dyn_BACE_jacet_mid"),Transform=(Rotation=(X=0,Y=0,Z=0,W=1),Translation=(X=0,Y=0,Z=0),Scale3D=(X=1,Y=1,Z=1))))
    WeightMinimum=0
    ItemToModify.0.Transform=&RigVMFunction_MathTransformFromSRT(Location=(X=0,Y=0,Z=0), Rotation=(X=0,Y=0,Z=15), RotationOrder=XYZ, Scale=(X=1,Y=1,Z=1)).Transform
    Weight=@n1.OutputParam
   )
  ]}
]
hierarchy [
 root {loc=<0 0 0>, rot=<0 0 0 1>}
 pelvis {parent=root, loc=<0 -0.5301 98.0571>, rot=<0.0127 0.707 -0.0127 -0.707>}
 spine_01 {parent=pelvis, loc=<6.1208 -0.9328 0>, rot=<0 0 0.034 -0.9994>}
 spine_02 {parent=spine_01, loc=<6.5996 -0.2095 0>, rot=<0 0 0.0238 -0.9997>}
 spine_03 {parent=spine_02, loc=<6.4914 -0.1237 0>, rot=<0 0 0.0149 -0.9999>}
 spine_04 {parent=spine_03, loc=<6.9598 0.0794 0>, rot=<0 0 -0.0132 -0.9999>}
 spine_05 {parent=spine_04, loc=<6.9987 0 0>, rot=<0 0 -0.0015 -1>}
 attach_cape {parent=spine_05, loc=<9.3216 -8.4501 0>, rot=<0.0616 -0.7044 -0.0616 -0.7044>}
 neck_01 {parent=spine_05, loc=<11.3138 0.1529 0>, rot=<0 0 -0.2003 -0.9797>}
 neck_02 {parent=neck_01, loc=<4.5627 0 0>, rot=<0 0 0.0504 -0.9987>}
 head {parent=neck_02, loc=<5.3914 0 0>, rot=<0 0 0.115 -0.9934>}
 neck_lod_2_corrected {parent=neck_01, loc=<0 0 0>, rot=<0 0 0 1>}
 attach_backpack {parent=spine_05, loc=<0 -8.35 0>, rot=<-0.0093 -0.707 0.0093 -0.707>}
 clavicle_l {parent=spine_05, loc=<5.4831 -0.1053 -3.2946>, rot=<0.0465 -0.7289 -0.009 -0.6829>}
 upperarm_l {parent=clavicle_l, loc=<9.8658 -1.2136 -0.4105>, rot=<0.0164 -0.3245 -0.0269 -0.9454>}
 lowerarm_l {parent=upperarm_l, loc=<26.465 -0.8345 0.259>, rot=<0.0036 0.0161 -0.2298 -0.9731>}
 hand_l {parent=lowerarm_l, loc=<22.7355 -0.3203 -0.6105>, rot=<-0.6784 -0.0223 0.0735 -0.7307>}
 thumb_01_l {parent=hand_l, loc=<1.3118 -0.3191 -2.459>, rot=<0.6546 -0.4042 0.0664 -0.6354>}
 thumb_02_l {parent=thumb_01_l, loc=<5.9336 -0.0448 0.0401>, rot=<-0.0037 0.0213 0.298 -0.9543>}
 thumb_03_l {parent=thumb_02_l, loc=<2.2223 0.3682 -0.0907>, rot=<0.0118 -0.013 0.0066 -0.9998>}
 index_metacarpal_l {parent=hand_l, loc=<1.6535 0.0183 -2.96>, rot=<0.0158 -0.0124 -0.0173 -0.9996>}
 index_01_l {parent=index_metacarpal_l, loc=<8.3008 0.0357 -0.0524>, rot=<-0.0044 0.0485 0.0924 -0.9945>}
 index_02_l {parent=index_01_l, loc=<3.3587 -0.3004 0.0452>, rot=<-0.002 0.0027 0.3225 -0.9466>}
 index_03_l {parent=index_02_l, loc=<2.5578 -0.123 -0.0252>, rot=<-0.002 0.0067 0.2558 -0.9667>}
 middle_metacarpal_l {parent=hand_l, loc=<2.1856 -0.1725 -0.5984>, rot=<0.006 0.0168 -0.0329 -0.9993>}
 middle_01_l {parent=middle_metacarpal_l, loc=<7.3275 -0.2447 0.0129>, rot=<-0.0208 0.0042 0.0858 -0.9961>}
 middle_02_l {parent=middle_01_l, loc=<4.0123 0.1401 0.1175>, rot=<-0.0218 0.0258 0.3512 -0.9357>}
 middle_03_l {parent=middle_02_l, loc=<2.6077 0.0534 0.1373>, rot=<0.0082 0.0188 0.1708 -0.9851>}
 ring_metacarpal_l {parent=hand_l, loc=<2.2074 -0.3625 1.3638>, rot=<-0.0284 0.0722 -0.0458 -0.9959>}
 ring_01_l {parent=ring_metacarpal_l, loc=<7.0295 -0.6197 -0.1446>, rot=<-0.0489 -0.0252 0.1892 -0.9804>}
 ring_02_l {parent=ring_01_l, loc=<3.6174 0.4647 -0.1027>, rot=<0 0 0.2806 -0.9598>}
 ring_03_l {parent=ring_02_l, loc=<2.4152 0.114 0.0199>, rot=<0 0 0.1189 -0.9929>}
 pinky_metacarpal_l {parent=hand_l, loc=<1.7616 -0.9753 3.3814>, rot=<-0.0397 0.0814 -0.069 -0.9935>}
 pinky_01_l {parent=pinky_metacarpal_l, loc=<7.1709 -0.4831 -0.1094>, rot=<-0.0654 -0.0035 0.2242 -0.9723>}
 pinky_02_l {parent=pinky_01_l, loc=<2.8274 0.0893 -0.1043>, rot=<0.0037 0.0011 0.2854 -0.9584>}
 pinky_03_l {parent=pinky_02_l, loc=<2.2046 0.0307 -0.0078>, rot=<-0.0002 0.0098 0.1445 -0.9895>}
 weapon_l {parent=hand_l, loc=<-0.1974 -3.4111 0.3811>, rot=<0 0 0 1>}
 hand_attach_l {parent=hand_l, loc=<8.5 -3 0>, rot=<0.7071 0.7071 0 0>}
 deform_wrist_dn_base_l {parent=hand_l, loc=<0 -1.5326 0>, rot=<0 0 0 -1>}
 deform_wrist_dn_l {parent=deform_wrist_dn_base_l, loc=<0 0 0>, rot=<0 0 0 1>}
 deform_wrist_up_base_l {parent=hand_l, loc=<0 2.3009 0>, rot=<0 0 0 -1>}
 deform_wrist_up_l {parent=deform_wrist_up_base_l, loc=<0 0 0>, rot=<0 0 0 1>}
 dyn_hand_ARM_R {parent=hand_l, loc=<0.8497 1.9084 -0.161>, rot=<0.7481 0.0941 -0.0841 -0.6515>}
 lowerarm_twist_02_l {parent=lowerarm_l, loc=<10.6707 0 0>, rot=<0 0 0 1>}
 lowerarm_twist_01_l {parent=lowerarm_l, loc=<18.316 0 0>, rot=<0 0 0 1>}
 deform_elbow_in_base_l {parent=lowerarm_l, loc=<0.1834 3.6062 -0.2572>, rot=<0 0 0 -1>}
 deform_elbow_in_l {parent=deform_elbow_in_base_l, loc=<0 0 0>, rot=<0 0 0 1>}
 deform_elbow_out_base_l {parent=lowerarm_l, loc=<0 -4 0>, rot=<0 0 0 1>}
 deform_elbow_out_l {parent=deform_elbow_out_base_l, loc=<0 0 0>, rot=<0 0 0 1>}
 upperarm_twist_01_l {parent=upperarm_l, loc=<7.2135 -0.1134 0>, rot=<0 0 0 1>}
 deform_upperarm_bk_base_l {parent=upperarm_twist_01_l, loc=<-6.4053 -4.3746 0.314>, rot=<0 0 0 -1>}
 deform_upperarm_bk_l {parent=deform_upperarm_bk_base_l, loc=<0 0 0>, rot=<0 0 0 1>}
 deform_upperarm_fr_base_l {parent=upperarm_twist_01_l, loc=<-5.6312 5.092 0.3354>, rot=<0 0 0 -1>}
 deform_upperarm_fr_l {parent=deform_upperarm_fr_base_l, loc=<0 0 0>, rot=<0 0 0 1>}
 deform_upperarm_up_base_l {parent=upperarm_twist_01_l, loc=<-8.891 0.4825 3.9517>, rot=<0 0 0 -1>}
 deform_upperarm_up_l {parent=deform_upperarm_up_base_l, loc=<0 0 0>, rot=<0 0 0 1>}
 dyn_puf_5_BACE {parent=upperarm_twist_01_l, loc=<5.1037 -4.573 2.4836>, rot=<0.0633 -0.3857 0.9182 -0.0643>}
 dyn_puf_5 {parent=dyn_puf_5_BACE, loc=<0 0 0>, rot=<0 0 0 1>}
 dyn_puf_5_end {parent=dyn_puf_5, loc=<10.8099 0.0001 -0.0001>, rot=<0 0 0 1>}
 dyn_puf_4_BACE {parent=upperarm_twist_01_l, loc=<5.0096 -2.4327 3.9755>, rot=<0.0589 -0.1389 0.9885 0.0077>}
 dyn_puf_4 {parent=dyn_puf_4_BACE, loc=<0 0 0>, rot=<0 0 0 1>}
 dyn_puf_4_end {parent=dyn_puf_4, loc=<11.5733 0 0>, rot=<0 0 0 1>}
 dyn_puf_3_BACE {parent=upperarm_twist_01_l, loc=<5.1373 0.6007 4.0201>, rot=<0.0801 0.1227 0.9886 0.0352>}
 dyn_puf_3 {parent=dyn_puf_3_BACE, loc=<0 0 0>, rot=<0 0 0 1>}
 dyn_puf_3_end {parent=dyn_puf_3, loc=<12.3422 -0.0002 -0.0005>, rot=<0 0 0 1>}
 dyn_puf_2_BACE {parent=upperarm_twist_01_l, loc=<5.0473 3.0735 2.7513>, rot=<0.0563 0.2893 0.9509 0.0941>}
 dyn_puf_2 {parent=dyn_puf_2_BACE, loc=<0 0 0>, rot=<0 0 0 1>}
 dyn_puf_2_end {parent=dyn_puf_2, loc=<10.565 -0.0004 -0.0003>, rot=<0 0 0 1>}
 dyn_puf_1_BACE {parent=upperarm_twist_01_l, loc=<5.1962 4.3104 0.3114>, rot=<0.0955 0.5428 0.8232 0.1365>}
 dyn_puf_1 {parent=dyn_puf_1_BACE, loc=<0 0 0>, rot=<0 0 0 1>}
 dyn_puf_1_end {parent=dyn_puf_1, loc=<9.165 -0.0004 -0.0003>, rot=<0 0 0 1>}
 dyn_puf_6_BACE {parent=upperarm_twist_01_l, loc=<5.2441 -5.3863 -0.2701>, rot=<0.0221 -0.5625 0.8232 -0.073>}
 dyn_puf_6 {parent=dyn_puf_6_BACE, loc=<0 0 0>, rot=<0 0 0 -1>}
 dyn_puf_6_end {parent=dyn_puf_6, loc=<8.6796 -0.0002 0.0001>, rot=<0 0 0 -1>}
 upperarm_twist_02_l {parent=upperarm_l, loc=<19.7998 -0.1134 0>, rot=<0 0 0 1>}
 dyn_Pin {parent=clavicle_l, loc=<5.2922 8.5508 -2.3265>, rot=<0.025 0.5079 -0.4257 -0.7485>}
 clavicle_r {parent=spine_05, loc=<5.4831 -0.1053 3.2946>, rot=<0.7289 0.0465 -0.6829 0.009>}
 upperarm_r {parent=clavicle_r, loc=<-9.8658 1.2136 0.4105>, rot=<0.0164 -0.3245 -0.0269 -0.9454>}
 lowerarm_r {parent=upperarm_r, loc=<-26.465 0.8345 -0.259>, rot=<0.0036 0.0161 -0.2298 -0.9731>}
 hand_r {parent=lowerarm_r, loc=<-22.7355 0.3203 0.6106>, rot=<-0.6784 -0.0223 0.0735 -0.7307>}
 thumb_01_r {parent=hand_r, loc=<-1.3118 0.3191 2.459>, rot=<0.6546 -0.4042 0.0664 -0.6354>}
 thumb_02_r {parent=thumb_01_r, loc=<-5.9336 0.0448 -0.0401>, rot=<-0.0037 0.0213 0.298 -0.9543>}
 thumb_03_r {parent=thumb_02_r, loc=<-2.2223 -0.3682 0.0907>, rot=<0.0118 -0.013 0.0066 -0.9998>}
 index_metacarpal_r {parent=hand_r, loc=<-1.6535 -0.0183 2.96>, rot=<0.0158 -0.0124 -0.0173 -0.9996>}
 index_01_r {parent=index_metacarpal_r, loc=<-8.3008 -0.0357 0.0524>, rot=<-0.0044 0.0485 0.0924 -0.9945>}
 index_02_r {parent=index_01_r, loc=<-3.3587 0.3004 -0.0452>, rot=<-0.002 0.0027 0.3225 -0.9466>}
 index_03_r {parent=index_02_r, loc=<-2.5578 0.123 0.0252>, rot=<-0.002 0.0067 0.2558 -0.9667>}
 middle_metacarpal_r {parent=hand_r, loc=<-2.1856 0.1725 0.5984>, rot=<0.006 0.0168 -0.0329 -0.9993>}
 middle_01_r {parent=middle_metacarpal_r, loc=<-7.3275 0.2447 -0.0129>, rot=<-0.0208 0.0042 0.0858 -0.9961>}
 middle_02_r {parent=middle_01_r, loc=<-4.0123 -0.1401 -0.1175>, rot=<-0.0218 0.0258 0.3512 -0.9357>}
 middle_03_r {parent=middle_02_r, loc=<-2.6077 -0.0534 -0.1373>, rot=<0.0082 0.0188 0.1708 -0.9851>}
 ring_metacarpal_r {parent=hand_r, loc=<-2.2074 0.3625 -1.3638>, rot=<-0.0284 0.0722 -0.0458 -0.9959>}
 ring_01_r {parent=ring_metacarpal_r, loc=<-7.0295 0.6197 0.1446>, rot=<-0.0489 -0.0252 0.1892 -0.9804>}
 ring_02_r {parent=ring_01_r, loc=<-3.6174 -0.4647 0.1027>, rot=<0 0 0.2806 -0.9598>}
 ring_03_r {parent=ring_02_r, loc=<-2.4152 -0.114 -0.0199>, rot=<0 0 0.1189 -0.9929>}
 pinky_metacarpal_r {parent=hand_r, loc=<-1.7616 0.9753 -3.3814>, rot=<-0.0397 0.0814 -0.069 -0.9935>}
 pinky_01_r {parent=pinky_metacarpal_r, loc=<-7.1709 0.4831 0.1094>, rot=<-0.0654 -0.0035 0.2242 -0.9723>}
 pinky_02_r {parent=pinky_01_r, loc=<-2.8274 -0.0893 0.1043>, rot=<0.0037 0.0011 0.2854 -0.9584>}
 pinky_03_r {parent=pinky_02_r, loc=<-2.2046 -0.0307 0.0078>, rot=<-0.0002 0.0098 0.1445 -0.9895>}
 weapon_r {parent=hand_r, loc=<0.1974 3.4111 -0.3811>, rot=<0 0 0 1>}
 hand_attach_r {parent=hand_r, loc=<-8.5001 2.9999 0>, rot=<0 0 -0.7071 -0.7071>}
 lowerarm_twist_02_r {parent=lowerarm_r, loc=<-10.6707 0 0>, rot=<0 0 0 -1>}
 lowerarm_twist_01_r {parent=lowerarm_r, loc=<-18.316 0 0>, rot=<0 0 0 -1>}
 deform_elbow_in_base_r {parent=lowerarm_r, loc=<-0.1834 -3.6062 0.2572>, rot=<0 0 0 -1>}
 deform_elbow_in_r {parent=deform_elbow_in_base_r, loc=<0 0 0>, rot=<0 0 0 1>}
 deform_elbow_out_base_r {parent=lowerarm_r, loc=<0 4 -0.0001>, rot=<0 0 0 -1>}
 deform_elbow_out_r {parent=deform_elbow_out_base_r, loc=<0 0 0>, rot=<0 0 0 1>}
 upperarm_twist_01_r {parent=upperarm_r, loc=<-7.2135 0.1134 0>, rot=<0 0 0 -1>}
 deform_upperarm_bk_base_r {parent=upperarm_twist_01_r, loc=<6.4053 4.3746 -0.314>, rot=<0 0 0 -1>}
 deform_upperarm_bk_r {parent=deform_upperarm_bk_base_r, loc=<0 0 0>, rot=<0 0 0 1>}
 deform_upperarm_fr_base_r {parent=upperarm_twist_01_r, loc=<5.6312 -5.092 -0.3354>, rot=<0 0 0 -1>}
 deform_upperarm_fr_r {parent=deform_upperarm_fr_base_r, loc=<0 0 0>, rot=<0 0 0 1>}
 deform_upperarm_up_base_r {parent=upperarm_twist_01_r, loc=<8.891 -0.4825 -3.9517>, rot=<0 0 0 -1>}
 deform_upperarm_up_r {parent=deform_upperarm_up_base_r, loc=<0 0 0>, rot=<0 0 0 1>}
 upperarm_twist_02_r {parent=upperarm_r, loc=<-19.7998 0.1134 0>, rot=<0 0 0 -1>}
 dyn_Belt_mid {parent=clavicle_r, loc=<-1.5397 0.0676 0.642>, rot=<-0.3703 -0.0539 0.9229 0.0905>}
 dyn_Belt_sholders {parent=dyn_Belt_mid, loc=<7.9027 -0.326 0.0812>, rot=<-0.6499 0.2795 -0.3852 -0.5925>}
 dyn_Belt_BACEtongue_up1 {parent=dyn_Belt_mid, loc=<1.7594 -7.3593 1.1273>, rot=<0.4522 -0.8855 -0.1044 -0.0234>}
 dyn_Belt_tongue_up {parent=dyn_Belt_BACEtongue_up1, loc=<0 0 0>, rot=<0 0 0 1>}
 dyn_Belt_tongue_up_end {parent=dyn_Belt_tongue_up, loc=<3.3913 0 0>, rot=<0 0 0 1>}
 dyn_Belt_BACE {parent=spine_05, loc=<-0.2342 1.5488 -1.0696>, rot=<-0.0687 0.4162 0.0756 -0.9035>}
 dyn_Belt_start {parent=dyn_Belt_BACE, loc=<0 0 0>, rot=<0 0 0 1>}
 dyn_BACE_jacet_front {parent=spine_01, loc=<3.6417 9.4133 -4.1905>, rot=<-0.0704 -0.9638 0.1901 -0.1734>}
 dyn_jacet_front_1 {parent=dyn_BACE_jacet_front, loc=<0 0 0>, rot=<0 0 0 1>}
 dyn_jacet_front_2 {parent=dyn_jacet_front_1, loc=<5.4669 0 0>, rot=<0 0 0 1>}
 dyn_jacet_front_3 {parent=dyn_jacet_front_2, loc=<5.0298 -0.0001 0.0001>, rot=<0 0 0 1>}
 dyn_BACE_jacet_mid {parent=spine_01, loc=<4.8486 1.4738 -9.4484>, rot=<-0.1282 -0.6824 0.6826 -0.2278>}
 dyn_jacet_mid_1 {parent=dyn_BACE_jacet_mid, loc=<0 0 0>, rot=<0 0 0 1>}
 dyn_jacet_mid_2 {parent=dyn_jacet_mid_1, loc=<6.166 -0.0001 0.0002>, rot=<0 0 0 1>}
 dyn_BACE_jacet_down {parent=spine_01, loc=<4.6905 -5.9228 0.0731>, rot=<-0.0225 -0.0158 0.979 -0.2021>}
 dyn_jacet_down_1 {parent=dyn_BACE_jacet_down, loc=<0 0 0>, rot=<0 0 0 1>}
 dyn_jacet_down_2 {parent=dyn_jacet_down_1, loc=<5.4498 0 -0.0003>, rot=<0 0 0 1>}
 dyn_jacet_down_3 {parent=dyn_jacet_down_2, loc=<4.9358 0 0.0002>, rot=<0 0 0 1>}
 dyn_BACE_jacet_mid_down {parent=spine_01, loc=<4.9588 -4.4586 -6.2354>, rot=<-0.107 -0.2369 0.9428 -0.2087>}
 dyn_jacet_mid_down_1 {parent=dyn_BACE_jacet_mid_down, loc=<0 0 0>, rot=<0 0 0 1>}
 dyn_jacet_mid_down_2 {parent=dyn_jacet_mid_down_1, loc=<5.4498 0 -0.0003>, rot=<0 0 0 1>}
 dyn_BACE_jacet_front_mid {parent=spine_01, loc=<4.7023 5.8073 -8.3941>, rot=<-0.1346 -0.8513 0.4847 -0.1493>}
 dyn_jacet_front_mid_1 {parent=dyn_BACE_jacet_front_mid, loc=<0 0 0>, rot=<0 0 0 1>}
 dyn_jacet_front_mid_2 {parent=dyn_jacet_front_mid_1, loc=<6.2849 0.021 0.3061>, rot=<0 0 0 1>}
 dyn_badge_down {parent=spine_01, loc=<5.7689 10.2479 -0.6698>, rot=<0.4919 0.508 0.508 0.4919>}
 thigh_l {parent=pelvis, loc=<-1.0855 -0.2109 -8.7053>, rot=<0.0009 0.0355 0.0155 -0.9993>}
 calf_l {parent=thigh_l, loc=<-42.5284 -0.0019 -0.7297>, rot=<0.0006 -0.0161 -0.0391 -0.9991>}
 foot_l {parent=calf_l, loc=<-46.7401 -2.5014 -0.592>, rot=<0.0427 -0.0238 0.0353 -0.9982>}
 ball_l {parent=foot_l, loc=<-3.5658 10.8066 -0.1544>, rot=<0.0001 -0.0001 -0.7063 -0.7079>}
 calf_twist_02_l {parent=calf_l, loc=<-19.1546 -1.7943 -0.2638>, rot=<0.0237 0.0095 -0.0035 -0.9997>}
 calf_twist_01_l {parent=calf_l, loc=<-35.8483 -1.9197 -0.5794>, rot=<0.0237 0.0095 -0.0035 -0.9997>}
 deform_knee_in_base_l {parent=calf_l, loc=<0.8031 -5.0231 0.2456>, rot=<0 0 0 -1>}
 deform_knee_in_l {parent=deform_knee_in_base_l, loc=<0 0 0>, rot=<0 0 0 1>}
 deform_knee_out_base_l {parent=calf_l, loc=<0 4 0>, rot=<0 0 0 -1>}
 deform_knee_out_l {parent=deform_knee_out_base_l, loc=<0 0 0>, rot=<0 0 0 1>}
 thigh_twist_01_l {parent=thigh_l, loc=<-27.952 0 0>, rot=<0.0013 0 0 -1>}
 deform_groin_base_l {parent=thigh_l, loc=<0.4801 5.6776 -0.2486>, rot=<0 0 0 -1>}
 deform_groin_l {parent=deform_groin_base_l, loc=<0 0 0>, rot=<0 0 0 1>}
 thigh_r {parent=pelvis, loc=<-1.0855 -0.2109 8.7053>, rot=<0.0355 -0.0009 0.9993 0.0155>}
 calf_r {parent=thigh_r, loc=<42.5284 0.0019 0.7297>, rot=<0.0006 -0.0161 -0.0391 -0.9991>}
 foot_r {parent=calf_r, loc=<46.7401 2.5014 0.592>, rot=<0.0427 -0.0238 0.0353 -0.9982>}
 ball_r {parent=foot_r, loc=<3.5658 -10.8066 0.1544>, rot=<0.0001 -0.0001 -0.7063 -0.7079>}
 calf_twist_02_r {parent=calf_r, loc=<19.1546 1.7943 0.2638>, rot=<0.0237 0.0095 -0.0035 -0.9997>}
 calf_twist_01_r {parent=calf_r, loc=<35.8483 1.9197 0.5794>, rot=<0.0237 0.0095 -0.0035 -0.9997>}
 deform_knee_in_base_r {parent=calf_r, loc=<-0.8031 5.0231 -0.2457>, rot=<0 0 0 -1>}
 deform_knee_in_r {parent=deform_knee_in_base_r, loc=<0 0 0>, rot=<0 0 0 1>}
 deform_knee_out_base_r {parent=calf_r, loc=<0 -4 0>, rot=<0 0 0 -1>}
 deform_knee_out_r {parent=deform_knee_out_base_r, loc=<0 0 0>, rot=<0 0 0 1>}
 thigh_twist_01_r {parent=thigh_r, loc=<27.952 0 0>, rot=<0.0013 0 0 -1>}
 deform_groin_base_r {parent=thigh_r, loc=<-0.4801 -5.6776 0.2486>, rot=<0 0 0 -1>}
 deform_groin_r {parent=deform_groin_base_r, loc=<0 0 0>, rot=<0 0 0 1>}
 dyn_beld1 {parent=pelvis, loc=<8.4822 0.6908 -8.4859>, rot=<-0.0235 0.7871 -0.0371 -0.6153>}
 dyn_BACE_scale_trans {parent=dyn_beld1, loc=<19.5001 0.0001 0>, rot=<0 0 0 1>}
 dyn_BACE_chain_small {parent=dyn_BACE_scale_trans, loc=<3.4705 2.0957 -0.7241>, rot=<0.111 0.1555 0.018 -0.9814>}
 dyn_chain_small_0 {parent=dyn_BACE_chain_small, loc=<0 0 0>, rot=<0 0 0 1>}
 dyn_chain_small_1 {parent=dyn_chain_small_0, loc=<3.0975 0.0283 0.0722>, rot=<0.2206 0.068 -0.2698 -0.9348>}
 dyn_sword_0 {parent=dyn_chain_small_1, loc=<1.8479 -0.5073 1.08>, rot=<-0.4679 0.6207 0.3734 -0.5063>}
 dyn_chain_long_PARENT {parent=dyn_sword_0, loc=<6.9816 -0.914 2.4702>, rot=<0.1906 -0.8078 0.4854 0.2749>}
 dyn_BACE_trans_creplenie {parent=dyn_BACE_scale_trans, loc=<0 0 0>, rot=<0 0 0 1>}
 dyn_BACE_chain_long_kINEM {parent=dyn_BACE_scale_trans, loc=<3.0712 -2.7552 -0.7324>, rot=<0.1881 0.4855 0.0469 -0.8525>}
 dyn_chain_long_3_KINEM_BACE {parent=dyn_BACE_scale_trans, loc=<2.729 -2.4863 -0.4175>, rot=<0.1881 0.4855 0.0469 -0.8525>}
 dyn_chain_long_3_KINEM {parent=dyn_chain_long_3_KINEM_BACE, loc=<0 0 0>, rot=<0 0 0 1>}
 dyn_chain_long_1_KINEM {parent=dyn_chain_long_3_KINEM, loc=<3.3259 -0.0546 -0.2106>, rot=<0.0323 -0.0273 0.0156 -0.999>}
 dyn_chain_long_2_KINEM {parent=dyn_chain_long_1_KINEM, loc=<1.9161 -0.1099 0.1473>, rot=<-0.0181 0.0396 0.0179 -0.9989>}
 dyn_chain_long_end1_KINEM {parent=dyn_chain_long_2_KINEM, loc=<3.212 0.066 -0.0733>, rot=<-0.1141 -0.004 -0.0005 -0.9935>}
 dyn_BACE_beld_skale {parent=dyn_beld1, loc=<0 0 0>, rot=<0 0 1 0>}
 dyn_BACE_beld_Transform {parent=dyn_beld1, loc=<0.7 0 0>, rot=<0 0 1 0>}
 deform_glute_base_l {parent=pelvis, loc=<-7.3466 -9.0213 -9.1245>, rot=<0.0172 0.031 -0.4477 -0.8935>}
 deform_glute_l {parent=deform_glute_base_l, loc=<0 0 0>, rot=<0 0 0 1>}
 deform_glute_base_r {parent=pelvis, loc=<-7.3464 -9.0213 9.1245>, rot=<0.031 -0.0172 0.8935 -0.4477>}
 deform_glute_r {parent=deform_glute_base_r, loc=<0 0 0>, rot=<0 0 0 1>}
 dyn_BACE_tongue_down {parent=pelvis, loc=<12.3222 -3.2997 9.1493>, rot=<0.5211 -0.4786 0.5318 -0.4654>}
 dyn_tongue_down {parent=dyn_BACE_tongue_down, loc=<0 0 0>, rot=<0 0 0 1>}
 dyn_tongue_down_end {parent=dyn_tongue_down, loc=<2 0 0>, rot=<0 0 0 1>}
]
variables [is sky_ {type=bool, default=True}]
