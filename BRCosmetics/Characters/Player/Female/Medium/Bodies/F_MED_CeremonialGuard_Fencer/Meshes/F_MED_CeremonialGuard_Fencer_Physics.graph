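# Physics asset for the female medium CeremonialGuard Fencer body, simulating six "puf" costume bone chains on the left upper arm.
# Each chain runs from a small BACE capsule locked to upperarm_l, through a middle capsule, to an end capsule locked to the middle.
# Only the BACE-to-middle joints are left unlocked, so each puff moves as one piece about its base.
# An upper-arm capsule and an unconstrained spherical head body complete the setup.
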PhysicsAsset /BRCosmetics/Characters/Player/Female/Medium/Bodies/F_MED_CeremonialGuard_Fencer/Meshes/F_MED_CeremonialGuard_Fencer_Physics
bodies {
 dyn_puf_1 [{shapeType=Capsule, length=5.5623, radius=1.0183, center=<8.4146 0 1.5327>, rotation=<rot -72.7505 -179.9999 179.9999>}]
 dyn_puf_1_BACE [{shapeType=Capsule, length=0.505, radius=0.505, shapeName=dyn_puf_1_BACE_capsule, rotation=<rot 0 -90 0>}]
 dyn_puf_1_end [{shapeType=Capsule, length=3.2586, radius=1.4886, shapeName=dyn_puf_1_end_capsule, center=<0.1155 0 0.4127>, rotation=<rot 0 -90 -74.3708>}]
 dyn_puf_2 [{shapeType=Capsule, length=4.2224, radius=1.2597, shapeName=dyn_puf_2_capsule, center=<10.1931 0 1.7004>, rotation=<rot 0 -90 87.2222>}]
 dyn_puf_2_BACE [{shapeType=Capsule, length=0.505, radius=0.505, shapeName=dyn_puf_2_BACE_capsule, rotation=<rot 0 -90 0>}]
 dyn_puf_2_end [{shapeType=Capsule, length=3.2586, radius=1.4886, shapeName=dyn_puf_2_end_capsule, center=<0.3883 0 1.3881>, rotation=<rot 0 -90 -74.3708>}]
 dyn_puf_3 [{shapeType=Capsule, length=3.5984, radius=1.8838, shapeName=dyn_puf_3_capsule, center=<11.4176 0 1.9836>, rotation=<rot 0 -90 87.2222>}]
 dyn_puf_3_BACE [{shapeType=Capsule, length=0.505, radius=0.505, shapeName=dyn_puf_3_BACE_capsule, rotation=<rot 0 -90 0>}]
 dyn_puf_3_end [{shapeType=Capsule, length=3.2586, radius=1.4886, shapeName=dyn_puf_3_end_capsule, center=<0.3283 0 1.1736>, rotation=<rot 0 -90 -74.3709>}]
 dyn_puf_4 [{shapeType=Capsule, length=3.7275, radius=1.7546, shapeName=dyn_puf_4_capsule, center=<10.1677 0 2.225>, rotation=<rot 0 -90 87.2222>}]
 dyn_puf_4_BACE [{shapeType=Capsule, length=0.505, radius=0.505, shapeName=dyn_puf_4_BACE_capsule, rotation=<rot 0 -90 0>}]
 dyn_puf_4_end [{shapeType=Capsule, length=3.2586, radius=1.4886, shapeName=dyn_puf_4_end_capsule, center=<0.3283 0 1.1736>, rotation=<rot 0 -90 -74.3709>}]
 dyn_puf_5 [{shapeType=Capsule, length=4.2224, radius=1.2597, shapeName=dyn_puf_5_capsule, center=<10.1996 0 1.5667>, rotation=<rot 0 -90 87.2222>}]
 dyn_puf_5_BACE [{shapeType=Capsule, length=0.505, radius=0.505, shapeName=dyn_puf_5_BACE_capsule, rotation=<rot 0 -90 0>}]
 dyn_puf_5_end [{shapeType=Capsule, length=3.2586, radius=1.4886, shapeName=dyn_puf_5_end_capsule, center=<0.1661 0 0.5936>, rotation=<rot 0 -90 -74.3709>}]
 dyn_puf_6 [{shapeType=Capsule, length=4.2224, radius=1.2597, shapeName=dyn_puf_6_capsule, center=<10.1931 0 1.7004>, rotation=<rot 0 -90 87.2222>}]
 dyn_puf_6_BACE [{shapeType=Capsule, length=0.505, radius=0.505, shapeName=dyn_puf_6_BACE_capsule, rotation=<rot 0 -90 0>}]
 dyn_puf_6_end [{
   shapeType=Capsule
   length=3.5187
   radius=1.6884
   shapeName=dyn_puf_6_end_capsule
   center=<0.4105 0.8423 0.6463>
   rotation=<rot 0 79.8466 -87.0081>
  }]
 head [{shapeType=Capsule, length=0, radius=7.7533, shapeName=head_capsule, rotation=<rot 90 0 0>}]
 upperarm_l [{shapeType=Capsule, length=4.6458, radius=4.7718, center=<4.0538 0 0>, rotation=<rot 0 -90 90>}]
}
constraints [
 {
  bone1Name=dyn_puf_3_end
  bone2Name=dyn_puf_3
  swing1LimitDegrees=0
  swing1Motion=Locked
  swing2LimitDegrees=5
  swing2Motion=Locked
  twistLimitDegrees=10
  twistMotion=Locked
 }
 {
  bone1Name=dyn_puf_4_end
  bone2Name=dyn_puf_4
  swing1LimitDegrees=0
  swing1Motion=Locked
  swing2LimitDegrees=5
  swing2Motion=Locked
  twistLimitDegrees=10
  twistMotion=Locked
 }
 {
  bone1Name=dyn_puf_5_end
  bone2Name=dyn_puf_5
  swing1LimitDegrees=0
  swing1Motion=Locked
  swing2LimitDegrees=5
  swing2Motion=Locked
  twistLimitDegrees=10
  twistMotion=Locked
 }
 {
  bone1Name=dyn_puf_2_end
  bone2Name=dyn_puf_2
  swing1LimitDegrees=0
  swing1Motion=Locked
  swing2LimitDegrees=5
  swing2Motion=Locked
  twistLimitDegrees=10
  twistMotion=Locked
 }
 {
  bone1Name=dyn_puf_1_end
  bone2Name=dyn_puf_1
  swing1LimitDegrees=0
  swing1Motion=Locked
  swing2LimitDegrees=5
  swing2Motion=Locked
  twistLimitDegrees=10
  twistMotion=Locked
 }
 {
  bone1Name=dyn_puf_1_BACE
  bone2Name=upperarm_l
  swing1LimitDegrees=45
  swing1Motion=Locked
  swing2LimitDegrees=45
  swing2Motion=Locked
  twistLimitDegrees=45
  twistMotion=Locked
 }
 {bone1Name=dyn_puf_1, bone2Name=dyn_puf_1_BACE, swing1LimitDegrees=0, swing2LimitDegrees=3, twistLimitDegrees=0}
 {
  bone1Name=dyn_puf_2_BACE
  bone2Name=upperarm_l
  swing1LimitDegrees=45
  swing1Motion=Locked
  swing2LimitDegrees=45
  swing2Motion=Locked
  twistLimitDegrees=45
  twistMotion=Locked
 }
 {bone1Name=dyn_puf_2, bone2Name=dyn_puf_2_BACE, swing1LimitDegrees=0, swing2LimitDegrees=3, twistLimitDegrees=0}
 {
  bone1Name=dyn_puf_3_BACE
  bone2Name=upperarm_l
  swing1LimitDegrees=45
  swing1Motion=Locked
  swing2LimitDegrees=45
  swing2Motion=Locked
  twistLimitDegrees=45
  twistMotion=Locked
 }
 {bone1Name=dyn_puf_3, bone2Name=dyn_puf_3_BACE, swing1LimitDegrees=0, swing2LimitDegrees=3, twistLimitDegrees=0}
 {
  bone1Name=dyn_puf_4_BACE
  bone2Name=upperarm_l
  swing1LimitDegrees=45
  swing1Motion=Locked
  swing2LimitDegrees=45
  swing2Motion=Locked
  twistLimitDegrees=45
  twistMotion=Locked
 }
 {bone1Name=dyn_puf_4, bone2Name=dyn_puf_4_BACE, swing1LimitDegrees=0, swing2LimitDegrees=3, twistLimitDegrees=0}
 {
  bone1Name=dyn_puf_5_BACE
  bone2Name=upperarm_l
  swing1LimitDegrees=45
  swing1Motion=Locked
  swing2LimitDegrees=45
  swing2Motion=Locked
  twistLimitDegrees=45
  twistMotion=Locked
 }
 {bone1Name=dyn_puf_5, bone2Name=dyn_puf_5_BACE, swing1LimitDegrees=0, swing2LimitDegrees=3, twistLimitDegrees=0}
 {
  bone1Name=dyn_puf_6_BACE
  bone2Name=upperarm_l
  swing1LimitDegrees=45
  swing1Motion=Locked
  swing2LimitDegrees=45
  swing2Motion=Locked
  twistLimitDegrees=45
  twistMotion=Locked
 }
 {bone1Name=dyn_puf_6, bone2Name=dyn_puf_6_BACE, swing1LimitDegrees=0, swing2LimitDegrees=3, twistLimitDegrees=0}
 {
  bone1Name=dyn_puf_6_end
  bone2Name=dyn_puf_6
  swing1LimitDegrees=45
  swing1Motion=Locked
  swing2LimitDegrees=11
  swing2Motion=Locked
  twistLimitDegrees=45
  twistMotion=Locked
 }
]
