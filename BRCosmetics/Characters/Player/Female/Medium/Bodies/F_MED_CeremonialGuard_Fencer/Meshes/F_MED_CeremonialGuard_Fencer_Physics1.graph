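# Physics asset for the female medium CeremonialGuard Fencer cosmetic body, covering a belt-hung sword with a short and a long chain.
# Both chains run from belt anchor bodies to the sword, closing a constraint loop tied to the pelvis.
# Twist is locked on every joint, and swing is open only at a few chain joints, leaving other links rigid.
# Pelvis and right thigh capsules add body volume, while several helper bodies have zero size.
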PhysicsAsset /BRCosmetics/Characters/Player/Female/Medium/Bodies/F_MED_CeremonialGuard_Fencer/Meshes/F_MED_CeremonialGuard_Fencer_Physics1
bodies {
 dyn_BACE_beld_Transform [{shapeType=Capsule, length=0, radius=0}]
 dyn_BACE_beld_skale [{shapeType=Capsule, length=0, radius=0}]
 dyn_BACE_chain_small [{shapeType=Capsule, length=0.505, radius=0.505, shapeName=dyn_BACE_chain_small_capsule, rotation=<rot 0 -90 0>}]
 dyn_BACE_scale_trans [{shapeType=Capsule, length=0.505, radius=0.505, shapeName=dyn_BACE_scale_trans_capsule, rotation=<rot 0 -90 0>}]
 dyn_beld1 [{shapeType=Capsule, length=0, radius=0}]
 dyn_chain_long_1_KINEM [{
   shapeType=Capsule
   length=1.1264
   radius=0.9953
   shapeName=dyn_chain_long_1_KINEM_capsule
   center=<0.9094 -0.0804 0.1922>
   rotation=<rot 0 89.3687 -86.247>
  }]
 dyn_chain_long_2_KINEM [{
   shapeType=Capsule
   length=1.6076
   radius=0.5
   shapeName=dyn_chain_long_2_KINEM_capsule
   center=<1.591 -0.0182 -0.0201>
   rotation=<rot 0 -90.001 -87.3095>
  }]
 dyn_chain_long_3_KINEM [{
   shapeType=Capsule
   length=1.2297
   radius=0.8891
   shapeName=dyn_chain_long_3_KINEM_capsule
   center=<1.6234 -0.0988 -0.0454>
   rotation=<rot 0 -91.5284 -87.0227>
  }]
 dyn_chain_long_3_KINEM_BACE [{shapeType=Capsule, length=0.505, radius=0.505, shapeName=dyn_chain_long_3_KINEM_BACE_capsule, rotation=<rot 0 -90 0>}]
 dyn_chain_long_PARENT [{shapeType=Capsule, length=0, radius=0, shapeName=dyn_chain_long_PARENT_capsule, rotation=<rot 0 -90 94.5162>}]
 dyn_chain_long_end1_KINEM [{shapeType=Capsule, length=0, radius=0, shapeName=dyn_chain_long_end1_KINEM_capsule, rotation=<rot 0 -90 0>}]
 dyn_chain_small_0 [{shapeType=Box, extentX=4.1904, extentY=2.5618, extentZ=1.2976, shapeName=dyn_chain_small_0_box, center=<1.6808 0 0>}]
 dyn_chain_small_1 [
  {shapeType=Sphere, radius=7.4706, shapeName=dyn_chain_small_1_sphere, center=<2.9741 4.7882 -0.4776>}
  {
   shapeType=Capsule
   length=144.7746
   radius=1.3682
   shapeName=dyn_chain_small_1_capsule
   center=<1.9147 0 0>
   rotation=<rot -76.7188 -90.5124 -3.7544>
  }
 ]
 dyn_sword_0 [{
   shapeType=Capsule
   length=46.4668
   radius=0
   shapeName=dyn_sword_0_capsule
   center=<26.3019 -2.0974 0.5334>
   rotation=<rot 0 87.9383 -85.8844>
  }]
 pelvis [{
   shapeType=Capsule
   length=8.85
   radius=9.1954
   shapeName=pelvis_capsule
   center=<-2.581 -0.6903 4.7047>
   rotation=<rot 3.1732 110.2595 -104.6026>
  }]
 thigh_twist_01_r [{
   shapeType=Capsule
   length=27.3935
   radius=7.8329
   shapeName=thigh_twist_01_r_capsule
   center=<-11.294 -0.056 -1.0363>
   rotation=<rot 15.3661 -90.0001 -92.5957>
  }]
}
constraints [
 {
  bone1Name=dyn_chain_small_1
  bone2Name=dyn_chain_small_0
  swing1LimitDegrees=45
  swing1Motion=Locked
  swing2LimitDegrees=45
  swing2Motion=Locked
  twistLimitDegrees=45
  twistMotion=Locked
 }
 {
  bone1Name=dyn_chain_long_1_KINEM
  bone2Name=dyn_chain_long_3_KINEM
  swing1LimitDegrees=20
  swing2LimitDegrees=10
  twistLimitDegrees=45
  twistMotion=Locked
 }
 {
  bone1Name=dyn_chain_long_2_KINEM
  bone2Name=dyn_chain_long_1_KINEM
  swing1LimitDegrees=20
  swing2LimitDegrees=15
  twistLimitDegrees=45
  twistMotion=Locked
 }
 {
  bone1Name=dyn_chain_small_0
  bone2Name=dyn_BACE_chain_small
  swing1LimitDegrees=2
  swing2LimitDegrees=10
  twistLimitDegrees=45
  twistMotion=Locked
 }
 {
  bone1Name=dyn_chain_long_PARENT
  bone2Name=dyn_sword_0
  swing1LimitDegrees=45
  swing1Motion=Locked
  swing2LimitDegrees=45
  swing2Motion=Locked
  twistLimitDegrees=45
  twistMotion=Locked
 }
 {
  bone1Name=dyn_chain_long_3_KINEM
  bone2Name=dyn_chain_long_3_KINEM_BACE
  swing1LimitDegrees=45
  swing2LimitDegrees=45
  twistLimitDegrees=45
  twistMotion=Locked
 }
 {
  bone1Name=dyn_BACE_scale_trans
  bone2Name=dyn_beld1
  swing1LimitDegrees=45
  swing1Motion=Locked
  swing2LimitDegrees=45
  swing2Motion=Locked
  twistLimitDegrees=45
  twistMotion=Locked
 }
 {
  bone1Name=thigh_twist_01_r
  bone2Name=pelvis
  swing1LimitDegrees=45
  swing1Motion=Locked
  swing2LimitDegrees=45
  swing2Motion=Locked
  twistLimitDegrees=45
  twistMotion=Locked
 }
 {
  bone1Name=dyn_chain_long_end1_KINEM
  bone2Name=dyn_chain_long_2_KINEM
  swing1LimitDegrees=45
  swing1Motion=Locked
  swing2LimitDegrees=45
  swing2Motion=Locked
  twistLimitDegrees=45
  twistMotion=Locked
 }
 {
  bone1Name=dyn_chain_long_2_KINEM
  bone2Name=dyn_chain_long_PARENT
  swing1LimitDegrees=45
  swing1Motion=Locked
  swing2LimitDegrees=45
  swing2Motion=Locked
  twistLimitDegrees=45
  twistMotion=Locked
 }
 {
  bone1Name=dyn_sword_0
  bone2Name=dyn_chain_small_1
  swing1LimitDegrees=45
  swing1Motion=Locked
  swing2LimitDegrees=45
  swing2Motion=Locked
  twistLimitDegrees=45
  twistMotion=Locked
 }
 {
  bone1Name=dyn_BACE_chain_small
  bone2Name=dyn_BACE_scale_trans
  swing1LimitDegrees=45
  swing1Motion=Locked
  swing2LimitDegrees=45
  swing2Motion=Locked
  twistLimitDegrees=45
  twistMotion=Locked
 }
 {
  bone1Name=dyn_beld1
  bone2Name=pelvis
  swing1LimitDegrees=45
  swing1Motion=Locked
  swing2LimitDegrees=45
  swing2Motion=Locked
  twistLimitDegrees=45
  twistMotion=Locked
 }
 {
  bone1Name=dyn_BACE_beld_Transform
  bone2Name=dyn_beld1
  swing1LimitDegrees=45
  swing1Motion=Locked
  swing2LimitDegrees=45
  swing2Motion=Locked
  twistLimitDegrees=45
  twistMotion=Locked
 }
 {
  bone1Name=dyn_BACE_beld_skale
  bone2Name=dyn_beld1
  swing1LimitDegrees=45
  swing1Motion=Locked
  swing2LimitDegrees=45
  swing2Motion=Locked
  twistLimitDegrees=45
  twistMotion=Locked
 }
 {
  bone1Name=dyn_chain_long_3_KINEM_BACE
  bone2Name=dyn_BACE_scale_trans
  swing1LimitDegrees=45
  swing1Motion=Locked
  swing2LimitDegrees=45
  swing2Motion=Locked
  twistLimitDegrees=45
  twistMotion=Locked
 }
 {
  bone1Name=dyn_chain_long_end1_KINEM
  bone2Name=dyn_chain_long_PARENT
  swing1LimitDegrees=45
  swing1Motion=Locked
  swing2LimitDegrees=45
  swing2Motion=Locked
  twistLimitDegrees=45
  twistMotion=Locked
 }
]
