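# Physics asset for the coat of the female medium Ceremonial Guard Fencer player cosmetic, defining its jacket-flap motion and body collision.
# Five short chains of box bodies (down, front, front-mid, mid, mid-down) hang from zero-size anchor bodies locked to the pelvis.
# Joints within the chains have tight per-joint swing and twist limits.
# Capsules on the pelvis and locked left thigh and twist bones provide body shapes; the right leg has none.
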PhysicsAsset /BRCosmetics/Characters/Player/Female/Medium/Bodies/F_MED_CeremonialGuard_Fencer/Meshes/F_MED_CeremonialGuard_Fencer_Physics_Coat
bodies {
 dyn_BACE_jacet_down [{shapeType=Capsule, length=0, radius=0}]
 dyn_BACE_jacet_front [{shapeType=Capsule, length=0, radius=0}]
 dyn_BACE_jacet_front_mid [{shapeType=Capsule, length=0, radius=0}]
 dyn_BACE_jacet_mid [{shapeType=Capsule, length=0, radius=0}]
 dyn_BACE_jacet_mid_down [{shapeType=Capsule, length=0, radius=0}]
 dyn_jacet_down_1 [{shapeType=Box, extentX=5.1219, extentY=2.9439, extentZ=7.5, shapeName=dyn_jacet_down_1_box, center=<2.284 0 0>}]
 dyn_jacet_down_2 [{shapeType=Box, extentX=5.1219, extentY=3.0109, extentZ=7.5, shapeName=dyn_jacet_down_2_box, center=<2.284 0 0>}]
 dyn_jacet_down_3 [{shapeType=Box, extentX=5.1219, extentY=2.9773, extentZ=7.5, shapeName=dyn_jacet_down_3_box, center=<2.284 0 0>}]
 dyn_jacet_front_1 [{shapeType=Box, extentX=5.1219, extentY=1.5086, extentZ=7.5, shapeName=dyn_jacet_front_1_box, center=<2.284 0 0>}]
 dyn_jacet_front_2 [{shapeType=Box, extentX=5.1219, extentY=0.9601, extentZ=7.5, shapeName=dyn_jacet_front_2_box, center=<2.284 0.1799 0>}]
 dyn_jacet_front_3 [{shapeType=Box, extentX=5.1219, extentY=1.6176, extentZ=7.5, shapeName=dyn_jacet_front_3_box, center=<2.284 1.4929 0>}]
 dyn_jacet_front_mid_1 [{shapeType=Box, extentX=4.9048, extentY=0.6495, extentZ=6.8374, shapeName=dyn_jacet_front_mid_1_box, center=<2.6785 0 0>}]
 dyn_jacet_front_mid_2 [{
   shapeType=Box
   extentX=9.6564
   extentY=0.6038
   extentZ=11.7958
   shapeName=dyn_jacet_front_mid_2_box
   center=<4.3559 0 -0.7918>
   rotation=<rot 0 0 8.6966>
  }]
 dyn_jacet_mid_1 [{shapeType=Box, extentX=5.1219, extentY=1.5086, extentZ=7.5, shapeName=dyn_jacet_mid_1_box, center=<2.284 0 0>}]
 dyn_jacet_mid_2 [{shapeType=Box, extentX=5.1219, extentY=1.5086, extentZ=7.5, shapeName=dyn_jacet_mid_2_box, center=<2.284 0 0>}]
 dyn_jacet_mid_down_1 [{shapeType=Box, extentX=5.1219, extentY=1.5086, extentZ=7.5, shapeName=dyn_jacet_mid_down_1_box, center=<2.284 0 0>}]
 dyn_jacet_mid_down_2 [{shapeType=Box, extentX=5.1219, extentY=1.5086, extentZ=7.5, shapeName=dyn_jacet_mid_down_2_box, center=<2.284 0 0>}]
 pelvis [{shapeType=Capsule, length=5.7691, radius=9.07}]
 thigh_l [
  {shapeType=Capsule, length=18.0745, radius=2, center=<-6.8234 4.5839 4.6508>, rotation=<rot 81.1165 177.7293 -165.6354>}
  {
   shapeType=Capsule
   length=12.7464
   radius=4.6866
   shapeName=thigh_l_capsule
   center=<-6.3905 2.2784 -0.9992>
   rotation=<rot -82.6554 -0.0006 0.0002>
  }
 ]
 thigh_twist_01_l [
  {shapeType=Capsule, length=10.2309, radius=8.0906, center=<21.0657 0.6902 1.156>, rotation=<rot -83.6057 -0.0541 -6.1406>}
  {
   shapeType=Capsule
   length=25.5631
   radius=8.0906
   shapeName=thigh_twist_01_l_capsule
   center=<14.4671 0.6619 1.4286>
   rotation=<rot -83.6057 -0.0543 -0.0737>
  }
 ]
}
constraints [
 {
  bone1Name=thigh_twist_01_l
  bone2Name=thigh_l
  swing1LimitDegrees=45
  swing1Motion=Locked
  swing2LimitDegrees=45
  swing2Motion=Locked
  twistLimitDegrees=45
  twistMotion=Locked
 }
 {
  bone1Name=dyn_BACE_jacet_down
  bone2Name=pelvis
  swing1LimitDegrees=45
  swing1Motion=Locked
  swing2LimitDegrees=45
  swing2Motion=Locked
  twistLimitDegrees=45
  twistMotion=Locked
 }
 {bone1Name=dyn_jacet_down_1, bone2Name=dyn_BACE_jacet_down, swing1LimitDegrees=11, swing2LimitDegrees=5, twistLimitDegrees=2}
 {bone1Name=dyn_jacet_down_2, bone2Name=dyn_jacet_down_1, swing1LimitDegrees=15, swing2LimitDegrees=5, twistLimitDegrees=2}
 {bone1Name=dyn_jacet_down_3, bone2Name=dyn_jacet_down_2, swing1LimitDegrees=10, swing2LimitDegrees=5, twistLimitDegrees=2}
 {
  bone1Name=dyn_BACE_jacet_front
  bone2Name=pelvis
  swing1LimitDegrees=45
  swing1Motion=Locked
  swing2LimitDegrees=45
  swing2Motion=Locked
  twistLimitDegrees=45
  twistMotion=Locked
 }
 {bone1Name=dyn_jacet_front_1, bone2Name=dyn_BACE_jacet_front, swing1LimitDegrees=3, swing2LimitDegrees=2, twistLimitDegrees=5}
 {bone1Name=dyn_jacet_front_2, bone2Name=dyn_jacet_front_1, swing1LimitDegrees=5, swing2LimitDegrees=2, twistLimitDegrees=3}
 {bone1Name=dyn_jacet_front_3, bone2Name=dyn_jacet_front_2, swing1LimitDegrees=15, swing2LimitDegrees=2, twistLimitDegrees=3}
 {
  bone1Name=dyn_BACE_jacet_front_mid
  bone2Name=pelvis
  swing1LimitDegrees=45
  swing1Motion=Locked
  swing2LimitDegrees=45
  swing2Motion=Locked
  twistLimitDegrees=45
  twistMotion=Locked
 }
 {bone1Name=dyn_jacet_front_mid_1, bone2Name=dyn_BACE_jacet_front_mid, swing1LimitDegrees=15, swing2LimitDegrees=0, twistLimitDegrees=2}
 {bone1Name=dyn_jacet_front_mid_2, bone2Name=dyn_jacet_front_mid_1, swing1LimitDegrees=11, swing2LimitDegrees=0, twistLimitDegrees=0}
 {
  bone1Name=dyn_BACE_jacet_mid
  bone2Name=pelvis
  swing1LimitDegrees=45
  swing1Motion=Locked
  swing2LimitDegrees=45
  swing2Motion=Locked
  twistLimitDegrees=45
  twistMotion=Locked
 }
 {bone1Name=dyn_jacet_mid_1, bone2Name=dyn_BACE_jacet_mid, swing1LimitDegrees=15, swing2LimitDegrees=2, twistLimitDegrees=1}
 {bone1Name=dyn_jacet_mid_2, bone2Name=dyn_jacet_mid_1, swing1LimitDegrees=30, swing2LimitDegrees=2, twistLimitDegrees=10}
 {
  bone1Name=dyn_BACE_jacet_mid_down
  bone2Name=pelvis
  swing1LimitDegrees=45
  swing1Motion=Locked
  swing2LimitDegrees=45
  swing2Motion=Locked
  twistLimitDegrees=45
  twistMotion=Locked
 }
 {bone1Name=dyn_jacet_mid_down_1, bone2Name=dyn_BACE_jacet_mid_down, swing1LimitDegrees=11, swing2LimitDegrees=5, twistLimitDegrees=2}
 {bone1Name=dyn_jacet_mid_down_2, bone2Name=dyn_jacet_mid_down_1, swing1LimitDegrees=15, swing2LimitDegrees=5, twistLimitDegrees=2}
 {
  bone1Name=thigh_l
  bone2Name=pelvis
  swing1LimitDegrees=45
  swing1Motion=Locked
  swing2LimitDegrees=45
  swing2Motion=Locked
  twistLimitDegrees=45
  twistMotion=Locked
 }
]
