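# Physics asset for the coat of the female medium Ceremonial Guard Fencer cosmetic body.
# Five short chains of box bodies (down, front, front-mid, mid, mid-down) hang from zero-size base bodies locked to the pelvis.
# Chain joints carry small, individually tuned swing and twist limit angles, mostly favoring swing1.
# Capsules on the pelvis, left thigh and left thigh twist form a locked body proxy, with no right-leg bodies.
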PhysicsAsset /BRCosmetics/Characters/Player/Female/Medium/Bodies/F_MED_CeremonialGuard_Fencer/Meshes/F_MED_CeremonialGuard_Fencer_Physics_Coat_standart
bodies {
 dyn_BACE_jacet_down [{shapeType=Capsule, length=0, radius=0}]
 dyn_BACE_jacet_front [{shapeType=Capsule, length=0, radius=0}]
 dyn_BACE_jacet_front_mid [{shapeType=Capsule, length=0, radius=0}]
 dyn_BACE_jacet_mid [{shapeType=Capsule, length=0, radius=0}]
 dyn_BACE_jacet_mid_down [{shapeType=Capsule, length=0, radius=0}]
 dyn_jacet_down_1 [{shapeType=Box, extentX=5.1219, extentY=1.5086, extentZ=7.5, shapeName=dyn_jacet_down_1_box, center=<2.284 0 0>}]
 dyn_jacet_down_2 [{shapeType=Box, extentX=5.1219, extentY=1.5086, extentZ=7.5, shapeName=dyn_jacet_down_2_box, center=<2.284 0 0>}]
 dyn_jacet_down_3 [{shapeType=Box, extentX=5.1219, extentY=1.5086, extentZ=7.5, shapeName=dyn_jacet_down_3_box, center=<2.284 0 0>}]
 dyn_jacet_front_1 [{shapeType=Box, extentX=7.6972, extentY=1.5086, extentZ=7.5, shapeName=dyn_jacet_front_1_box, center=<2.284 0 0>}]
 dyn_jacet_front_2 [{shapeType=Box, extentX=5.1219, extentY=0.7033, extentZ=7.5, shapeName=dyn_jacet_front_2_box, center=<2.284 0 0>}]
 dyn_jacet_front_3 [{shapeType=Box, extentX=5.1219, extentY=3.0327, extentZ=7.5, shapeName=dyn_jacet_front_3_box, center=<2.284 0 0>}]
 dyn_jacet_front_mid_1 [{shapeType=Box, extentX=5.1219, extentY=1.9086, extentZ=7.5, shapeName=dyn_jacet_front_mid_1_box, center=<2.284 0 0>}]
 dyn_jacet_front_mid_2 [{shapeType=Box, extentX=5.1219, extentY=0.9009, extentZ=12.6491, shapeName=dyn_jacet_front_mid_2_box, center=<2.284 0 0>}]
 dyn_jacet_mid_1 [{shapeType=Box, extentX=5.1219, extentY=1.5086, extentZ=7.5, shapeName=dyn_jacet_mid_1_box, center=<2.284 0 0>}]
 dyn_jacet_mid_2 [{shapeType=Box, extentX=5.1219, extentY=1.5086, extentZ=7.5, shapeName=dyn_jacet_mid_2_box, center=<2.284 0 0>}]
 dyn_jacet_mid_down_1 [{shapeType=Box, extentX=5.1219, extentY=1.5086, extentZ=7.5, shapeName=dyn_jacet_mid_down_1_box, center=<2.284 0 0>}]
 dyn_jacet_mid_down_2 [{shapeType=Box, extentX=5.1219, extentY=1.5086, extentZ=7.5, shapeName=dyn_jacet_mid_down_2_box, center=<2.284 0 0>}]
 pelvis [{shapeType=Capsule, length=5.7691, radius=10.2551, center=<1.0826 0 0>}]
 thigh_l [{shapeType=Capsule, length=18.0745, radius=7.7689, center=<-6.8247 2.5218 1.4868>, rotation=<rot 81.1165 177.7293 -165.6354>}]
 thigh_twist_01_l [{shapeType=Capsule, length=24.492, radius=9.1617, center=<14.9346 0.8331 0.4689>, rotation=<rot -83.6057 -0.0541 -6.1406>}]
}
constraints [
 {
  bone1Name=thigh_twist_01_l
  bone2Name=thigh_l
  swing1LimitDegrees=45
  swing1Motion=Locked
  swing2LimitDegrees=45
  swing2Motion=Locked
  twistLimitDegrees=45
  twistMotion=Locked
 }
 {
  bone1Name=dyn_BACE_jacet_down
  bone2Name=pelvis
  swing1LimitDegrees=45
  swing1Motion=Locked
  swing2LimitDegrees=45
  swing2Motion=Locked
  twistLimitDegrees=45
  twistMotion=Locked
 }
 {bone1Name=dyn_jacet_down_1, bone2Name=dyn_BACE_jacet_down, swing1LimitDegrees=11, swing2LimitDegrees=0, twistLimitDegrees=0}
 {bone1Name=dyn_jacet_down_2, bone2Name=dyn_jacet_down_1, swing1LimitDegrees=7, swing2LimitDegrees=0, twistLimitDegrees=0}
 {bone1Name=dyn_jacet_down_3, bone2Name=dyn_jacet_down_2, swing1LimitDegrees=5, swing2LimitDegrees=1, twistLimitDegrees=0}
 {
  bone1Name=dyn_BACE_jacet_front
  bone2Name=pelvis
  swing1LimitDegrees=45
  swing1Motion=Locked
  swing2LimitDegrees=45
  swing2Motion=Locked
  twistLimitDegrees=45
  twistMotion=Locked
 }
 {bone1Name=dyn_jacet_front_1, bone2Name=dyn_BACE_jacet_front, swing1LimitDegrees=2, swing2LimitDegrees=0, twistLimitDegrees=0}
 {bone1Name=dyn_jacet_front_2, bone2Name=dyn_jacet_front_1, swing1LimitDegrees=10, swing2LimitDegrees=2, twistLimitDegrees=2}
 {bone1Name=dyn_jacet_front_3, bone2Name=dyn_jacet_front_2, swing1LimitDegrees=5, swing2LimitDegrees=2, twistLimitDegrees=1}
 {
  bone1Name=dyn_BACE_jacet_front_mid
  bone2Name=pelvis
  swing1LimitDegrees=45
  swing1Motion=Locked
  swing2LimitDegrees=45
  swing2Motion=Locked
  twistLimitDegrees=45
  twistMotion=Locked
 }
 {bone1Name=dyn_jacet_front_mid_1, bone2Name=dyn_BACE_jacet_front_mid, swing1LimitDegrees=10, swing2LimitDegrees=4, twistLimitDegrees=1}
 {bone1Name=dyn_jacet_front_mid_2, bone2Name=dyn_jacet_front_mid_1, swing1LimitDegrees=20, swing2LimitDegrees=4, twistLimitDegrees=1}
 {
  bone1Name=dyn_BACE_jacet_mid
  bone2Name=pelvis
  swing1LimitDegrees=45
  swing1Motion=Locked
  swing2LimitDegrees=45
  swing2Motion=Locked
  twistLimitDegrees=45
  twistMotion=Locked
 }
 {bone1Name=dyn_jacet_mid_1, bone2Name=dyn_BACE_jacet_mid, swing1LimitDegrees=7, swing2LimitDegrees=2, twistLimitDegrees=5}
 {bone1Name=dyn_jacet_mid_2, bone2Name=dyn_jacet_mid_1, swing1LimitDegrees=7, swing2LimitDegrees=2, twistLimitDegrees=5}
 {
  bone1Name=dyn_BACE_jacet_mid_down
  bone2Name=pelvis
  swing1LimitDegrees=45
  swing1Motion=Locked
  swing2LimitDegrees=45
  swing2Motion=Locked
  twistLimitDegrees=45
  twistMotion=Locked
 }
 {bone1Name=dyn_jacet_mid_down_1, bone2Name=dyn_BACE_jacet_mid_down, swing1LimitDegrees=11, swing2LimitDegrees=2, twistLimitDegrees=0}
 {bone1Name=dyn_jacet_mid_down_2, bone2Name=dyn_jacet_mid_down_1, swing1LimitDegrees=15, swing2LimitDegrees=2, twistLimitDegrees=0}
 {
  bone1Name=thigh_l
  bone2Name=pelvis
  swing1LimitDegrees=45
  swing1Motion=Locked
  swing2LimitDegrees=45
  swing2Motion=Locked
  twistLimitDegrees=45
  twistMotion=Locked
 }
]
