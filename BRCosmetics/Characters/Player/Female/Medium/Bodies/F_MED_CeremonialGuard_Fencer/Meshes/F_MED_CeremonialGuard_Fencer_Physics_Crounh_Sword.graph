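# Physics asset for the female Ceremonial Guard Fencer body that sets up the hanging sword, belt and chains, apparently as a crouch variant.
# A belt body locked to the pelvis carries scale bodies, from which a short chain leads to the sword.
# A long chain links the scale to a parent body locked to the sword, closing a loop.
# Pelvis, right thigh and a wide, thin floor box on root supply body and ground collision shapes.
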PhysicsAsset /BRCosmetics/Characters/Player/Female/Medium/Bodies/F_MED_CeremonialGuard_Fencer/Meshes/F_MED_CeremonialGuard_Fencer_Physics_Crounh_Sword
bodies {
 dyn_BACE_beld_skale [{shapeType=Capsule, length=1, radius=1, rotation=<rot -47.7733 0.0002 -0.0002>}]
 dyn_BACE_chain_long_kINEM [{shapeType=Capsule, length=0.505, radius=0.505, shapeName=dyn_BACE_chain_long_kINEM_capsule, rotation=<rot 0 -90 0>}]
 dyn_BACE_chain_small [{shapeType=Capsule, length=0, radius=0}]
 dyn_BACE_scale_trans [{shapeType=Capsule, length=1, radius=1, center=<0.7563 0 -2.0993>, rotation=<rot -52.1824 -0.0001 0.0002>}]
 dyn_beld1 [{shapeType=Capsule, length=0, radius=0}]
 dyn_chain_long_1_KINEM [{
   shapeType=Capsule
   length=1.4638
   radius=0.6579
   shapeName=dyn_chain_long_1_KINEM_capsule
   center=<0.9094 -0.0804 0.1922>
   rotation=<rot 0 89.3687 -86.0557>
  }]
 dyn_chain_long_2_KINEM [{
   shapeType=Capsule
   length=1.6076
   radius=0.5
   shapeName=dyn_chain_long_2_KINEM_capsule
   center=<1.591 -0.0182 -0.0201>
   rotation=<rot 0 -90.001 -87.1182>
  }]
 dyn_chain_long_3_KINEM [{
   shapeType=Capsule
   length=1.5481
   radius=0.5707
   shapeName=dyn_chain_long_3_KINEM_capsule
   center=<1.6234 -0.0988 -0.0454>
   rotation=<rot 0 -91.5284 -86.8314>
  }]
 dyn_chain_long_3_KINEM_BACE [{shapeType=Capsule, length=0.505, radius=0.505, shapeName=dyn_chain_long_3_KINEM_BACE_capsule, rotation=<rot 0 -90 0>}]
 dyn_chain_long_PARENT [{shapeType=Capsule, length=0, radius=0, shapeName=dyn_chain_long_PARENT_capsule, rotation=<rot 0 -90 0>}]
 dyn_chain_long_end1_KINEM [{shapeType=Capsule, length=0, radius=0, shapeName=dyn_chain_long_end1_KINEM_capsule, rotation=<rot 0 -90 0>}]
 dyn_chain_small_0 [{
   shapeType=Box
   extentX=3.2114
   extentY=2.8383
   extentZ=0.6435
   shapeName=dyn_chain_small_0_box
   center=<0.7746 0.0401 0>
   rotation=<rot 0 2.9646 0>
  }]
 dyn_chain_small_1 [
  {shapeType=Sphere, radius=7.1766, shapeName=dyn_sword_0_sphere, center=<3.4826 7.0689 -1.4305>}
  {shapeType=Capsule, length=2.9931, radius=2.5581, rotation=<rot 0 0 -30.7113>}
  {
   shapeType=Capsule
   length=74.4225
   radius=1.8082
   shapeName=dyn_chain_small_1_capsule
   center=<-0.0333 -23.4803 5.7606>
   rotation=<rot 1.244 -5.0781 -76.2705>
  }
 ]
 dyn_sword_0 [{
   shapeType=Capsule
   length=46.4668
   radius=0
   shapeName=dyn_sword_0_capsule
   center=<26.3019 -2.0974 0.5334>
   rotation=<rot 0 87.9383 -85.8844>
  }]
 pelvis [
  {shapeType=Capsule, length=8.9815, radius=8.6515, center=<1.0938 1.4598 0.4183>}
  {shapeType=Capsule, length=10.3371, radius=7.2959, shapeName=pelvis_capsule, center=<6.3452 1.4598 1.0902>}
  {shapeType=Capsule, length=11.8696, radius=8.6214, shapeName=pelvis_capsule, center=<-7.6076 -2.9939 0.3774>}
  {
   shapeType=Capsule
   length=13.1952
   radius=7.2959
   shapeName=pelvis_capsule
   center=<-16.2979 -1.1663 8.1468>
   rotation=<rot -82.2564 179.8121 -11.5898>
  }
 ]
 root [
  {shapeType=Capsule, length=0.505, radius=0.505, shapeName=root_capsule, rotation=<rot 0 -90 0>}
  {shapeType=Box, extentX=274.9506, extentY=274.9506, extentZ=0.5, shapeName=root_box}
 ]
 thigh_twist_01_r [{
   shapeType=Capsule
   length=25.5631
   radius=8.0906
   shapeName=thigh_twist_01_l_capsule
   center=<-14.4671 -0.6619 -1.4286>
   rotation=<rot -83.6057 -0.0543 -179.8824>
  }]
}
constraints [
 {
  bone1Name=dyn_beld1
  bone2Name=pelvis
  swing1LimitDegrees=45
  swing1Motion=Locked
  swing2LimitDegrees=45
  swing2Motion=Locked
  twistLimitDegrees=45
  twistMotion=Locked
 }
 {bone1Name=dyn_BACE_beld_skale, bone2Name=dyn_beld1, swing1LimitDegrees=2, swing2LimitDegrees=15, twistLimitDegrees=2}
 {bone1Name=dyn_BACE_scale_trans, bone2Name=dyn_beld1, swing1LimitDegrees=5, swing2LimitDegrees=5, twistLimitDegrees=2}
 {
  bone1Name=dyn_BACE_chain_small
  bone2Name=dyn_BACE_scale_trans
  swing1LimitDegrees=10
  swing1Motion=Locked
  swing2LimitDegrees=5
  swing2Motion=Locked
  twistLimitDegrees=2
  twistMotion=Locked
 }
 {bone1Name=dyn_chain_small_0, bone2Name=dyn_BACE_chain_small, swing1LimitDegrees=5, swing2LimitDegrees=5, twistLimitDegrees=2}
 {bone1Name=dyn_chain_small_1, bone2Name=dyn_chain_small_0, swing1LimitDegrees=5, swing2LimitDegrees=5, twistLimitDegrees=2}
 {bone1Name=dyn_sword_0, bone2Name=dyn_chain_small_1, swing1LimitDegrees=0, swing2LimitDegrees=2, twistLimitDegrees=0}
 {
  bone1Name=thigh_twist_01_r
  bone2Name=pelvis
  swing1LimitDegrees=45
  swing1Motion=Locked
  swing2LimitDegrees=45
  swing2Motion=Locked
  twistLimitDegrees=45
  twistMotion=Locked
 }
 {
  bone1Name=pelvis
  bone2Name=root
  swing1LimitDegrees=45
  swing1Motion=Locked
  swing2LimitDegrees=45
  swing2Motion=Locked
  twistLimitDegrees=45
  twistMotion=Locked
 }
 {
  bone1Name=dyn_BACE_chain_long_kINEM
  bone2Name=dyn_BACE_scale_trans
  swing1LimitDegrees=45
  swing1Motion=Locked
  swing2LimitDegrees=45
  swing2Motion=Locked
  twistLimitDegrees=45
  twistMotion=Locked
 }
 {
  bone1Name=dyn_chain_long_3_KINEM
  bone2Name=dyn_chain_long_3_KINEM_BACE
  swing1LimitDegrees=45
  swing2LimitDegrees=45
  twistLimitDegrees=45
  twistMotion=Locked
 }
 {
  bone1Name=dyn_chain_long_1_KINEM
  bone2Name=dyn_chain_long_3_KINEM
  swing1LimitDegrees=20
  swing2LimitDegrees=10
  twistLimitDegrees=45
  twistMotion=Locked
 }
 {
  bone1Name=dyn_chain_long_2_KINEM
  bone2Name=dyn_chain_long_1_KINEM
  swing1LimitDegrees=20
  swing1Motion=Locked
  swing2LimitDegrees=15
  swing2Motion=Locked
  twistLimitDegrees=45
  twistMotion=Locked
 }
 {
  bone1Name=dyn_chain_long_end1_KINEM
  bone2Name=dyn_chain_long_2_KINEM
  swing1LimitDegrees=45
  swing1Motion=Locked
  swing2LimitDegrees=45
  swing2Motion=Locked
  twistLimitDegrees=45
  twistMotion=Locked
 }
 {
  bone1Name=dyn_chain_long_PARENT
  bone2Name=dyn_sword_0
  swing1LimitDegrees=45
  swing1Motion=Locked
  swing2LimitDegrees=45
  swing2Motion=Locked
  twistLimitDegrees=45
  twistMotion=Locked
 }
 {
  bone1Name=dyn_chain_long_end1_KINEM
  bone2Name=dyn_chain_long_PARENT
  swing1LimitDegrees=45
  swing1Motion=Locked
  swing2LimitDegrees=45
  swing2Motion=Locked
  twistLimitDegrees=45
  twistMotion=Locked
 }
 {bone1Name=dyn_chain_long_2_KINEM, bone2Name=dyn_chain_long_PARENT, swing1LimitDegrees=3, swing2LimitDegrees=3, twistLimitDegrees=1}
 {
  bone1Name=dyn_chain_long_3_KINEM_BACE
  bone2Name=dyn_BACE_scale_trans
  swing1LimitDegrees=45
  swing1Motion=Locked
  swing2LimitDegrees=45
  swing2Motion=Locked
  twistLimitDegrees=45
  twistMotion=Locked
 }
]
