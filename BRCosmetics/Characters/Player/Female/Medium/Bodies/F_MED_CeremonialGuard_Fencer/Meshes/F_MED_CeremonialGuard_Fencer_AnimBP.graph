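# Post-process style AnimBP for the female Ceremonial Guard Fencer cosmetic body, adding secondary motion to a pose copied from its parent mesh.
# Control Rigs feed stacked RigidBody sims for standing, crouch and sliding-named variants, then look-ats, constraint, AnimDynamics and deform rig.
# A layered blend lets front-end idle montages override sword and chain bones.
# Events update gravity, LOD and backbling flags, and toggle a material element by matching pickaxe.
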AnimBlueprint /BRCosmetics/Characters/Player/Female/Medium/Bodies/F_MED_CeremonialGuard_Fencer/Meshes/F_MED_CeremonialGuard_Fencer_AnimBP
target_skeleton=&/F_MED_CeremonialGuard_Fencer_Skeleton
graphs [
 AnimGraph {nodes [
   Anim Dynamics - Bone: dyn_BACE_tongue_down(Alpha=1, ExternalForce="0,0,0") {
    angularDampingOverride=0.1
    angularSpringConstant=5555
    bAngularSpring=true
    bOverrideAngularDamping=true
    bUseGravityOverride=true
    linearSpringConstant=500
    simulationSpace=BoneRelative
    alphaBoolBlend {blendOption=Linear}
    boundBone {boneName=dyn_BACE_tongue_down}
    componentAppliedLinearAccClamp=<10000 10000 10000>
    componentLinearAccScale=<0.5 0.5 0.5>
    componentLinearVelScale=<0.5 0.5 0.5>
    gravityOverride {z=-930}
    physicsBodyDefinitions [{
      boundBone {boneName=dyn_BACE_tongue_down}
      boxExtents=<3 5 5>
      constraintSetup {angularLimitsMax=<0 5 10>, angularLimitsMin=<0 -3 -7>, angularTarget=<0 2 0>}
      localJointOffset=<0.1 0 0>
     }]
    relativeSpaceBone {boneName=pelvis}
   }
   n1 = Anim Dynamics - Bone: dyn_puf_1_BACE(Alpha=1, ExternalForce="0,0,0") {
    angularDampingOverride=0.1
    angularSpringConstant=5555
    bAngularSpring=true
    bOverrideAngularDamping=true
    bUseGravityOverride=true
    simulationSpace=BoneRelative
    alphaBoolBlend {blendOption=Linear}
    boundBone {boneName=dyn_puf_1_BACE}
    componentAppliedLinearAccClamp=<10000 10000 10000>
    componentLinearAccScale=<0.5 0.5 0.5>
    componentLinearVelScale=<0.5 0.5 0.5>
    gravityOverride {z=-930}
    physicsBodyDefinitions [{
      boundBone {boneName=dyn_puf_1_BACE}
      boxExtents=<7 7 7>
      constraintSetup {angularLimitsMax=<10 7 0>, angularLimitsMin=<-10 -7 0>, angularTarget=<0 2 0>}
      localJointOffset=<0.1 0 0>
     }]
    relativeSpaceBone {boneName=upperarm_l}
   }
   n2 = Anim Dynamics - Bone: dyn_puf_2_BACE(ComponentPose=@n1, Alpha=1, ExternalForce="0,0,0") {
    angularDampingOverride=0.1
    angularSpringConstant=5555
    bAngularSpring=true
    bOverrideAngularDamping=true
    bUseGravityOverride=true
    simulationSpace=BoneRelative
    alphaBoolBlend {blendOption=Linear}
    boundBone {boneName=dyn_puf_2_BACE}
    componentAppliedLinearAccClamp=<10000 10000 10000>
    componentLinearAccScale=<0.5 0.5 0.5>
    componentLinearVelScale=<0.5 0.5 0.5>
    gravityOverride {z=-930}
    physicsBodyDefinitions [{
      boundBone {boneName=dyn_puf_2_BACE}
      boxExtents=<7 7 7>
      constraintSetup {angularLimitsMax=<10 7 0>, angularLimitsMin=<-10 -7 0>, angularTarget=<0 2 0>}
      localJointOffset=<0.1 0 0>
     }]
    relativeSpaceBone {boneName=upperarm_l}
   }
   n3 = Anim Dynamics - Bone: dyn_puf_3_BACE(ComponentPose=@n2, Alpha=1, ExternalForce="0,0,0") {
    angularDampingOverride=0.1
    angularSpringConstant=5555
    bAngularSpring=true
    bOverrideAngularDamping=true
    bUseGravityOverride=true
    linearSpringConstant=500
    simulationSpace=BoneRelative
    alphaBoolBlend {blendOption=Linear}
    boundBone {boneName=dyn_puf_3_BACE}
    componentAppliedLinearAccClamp=<10000 10000 10000>
    componentLinearAccScale=<0.5 0.5 0.5>
    componentLinearVelScale=<0.5 0.5 0.5>
    gravityOverride {z=-930}
    physicsBodyDefinitions [{
      boundBone {boneName=dyn_puf_3_BACE}
      boxExtents=<7 7 7>
      constraintSetup {angularLimitsMax=<10 3 0>, angularLimitsMin=<-5 -5 0>, angularTarget=<0 2 0>}
      localJointOffset=<0.1 0 0>
     }]
    relativeSpaceBone {boneName=upperarm_l}
   }
   n4 = Anim Dynamics - Bone: dyn_puf_4_BACE(ComponentPose=@n3, Alpha=1, ExternalForce="0,0,0") {
    angularDampingOverride=0.1
    angularSpringConstant=5555
    bAngularSpring=true
    bOverrideAngularDamping=true
    bUseGravityOverride=true
    simulationSpace=BoneRelative
    alphaBoolBlend {blendOption=Linear}
    boundBone {boneName=dyn_puf_4_BACE}
    componentAppliedLinearAccClamp=<10000 10000 10000>
    componentLinearAccScale=<0.5 0.5 0.5>
    componentLinearVelScale=<0.5 0.5 0.5>
    gravityOverride {z=-930}
    physicsBodyDefinitions [{
      boundBone {boneName=dyn_puf_4_BACE}
      boxExtents=<7 7 7>
      constraintSetup {angularLimitsMax=<10 3 0>, angularLimitsMin=<-5 -5 0>, angularTarget=<0 2 0>}
      localJointOffset=<0.1 0 0>
     }]
    relativeSpaceBone {boneName=upperarm_l}
   }
   n5 = Anim Dynamics - Bone: dyn_puf_5_BACE(ComponentPose=@n4, Alpha=1, ExternalForce="0,0,0") {
    angularDampingOverride=0.1
    angularSpringConstant=5555
    bAngularSpring=true
    bOverrideAngularDamping=true
    bUseGravityOverride=true
    linearSpringConstant=500
    simulationSpace=BoneRelative
    alphaBoolBlend {blendOption=Linear}
    boundBone {boneName=dyn_puf_5_BACE}
    componentAppliedLinearAccClamp=<10000 10000 10000>
    componentLinearAccScale=<0.5 0.5 0.5>
    componentLinearVelScale=<0.5 0.5 0.5>
    gravityOverride {z=-930}
    physicsBodyDefinitions [{
      boundBone {boneName=dyn_puf_5_BACE}
      boxExtents=<7 7 7>
      constraintSetup {angularLimitsMax=<10 7 0>, angularLimitsMin=<-10 -7 0>, angularTarget=<0 2 0>}
      localJointOffset=<0.1 0 0>
     }]
    relativeSpaceBone {boneName=upperarm_l}
   }
   Anim Dynamics - Bone: dyn_puf_6_BACE(ComponentPose=@n5, Alpha=1, ExternalForce="0,0,0") {
    angularDampingOverride=0.1
    angularSpringConstant=5555
    bAngularSpring=true
    bOverrideAngularDamping=true
    bUseGravityOverride=true
    simulationSpace=BoneRelative
    alphaBoolBlend {blendOption=Linear}
    boundBone {boneName=dyn_puf_6_BACE}
    componentAppliedLinearAccClamp=<10000 10000 10000>
    componentLinearAccScale=<0.5 0.5 0.5>
    componentLinearVelScale=<0.5 0.5 0.5>
    gravityOverride {z=-930}
    physicsBodyDefinitions [{
      boundBone {boneName=dyn_puf_6_BACE}
      boxExtents=<7 7 7>
      constraintSetup {angularLimitsMax=<10 7 0>, angularLimitsMin=<-10 -7 0>, angularTarget=<0 2 0>}
      localJointOffset=<0.1 0 0>
     }]
    relativeSpaceBone {boneName=upperarm_l}
   }
   n6 = "Transform (Modify) Bone - Bone: neck_lod_2_corrected"!(ComponentPose, bAlphaBoolEnabled:bool=Get LOD_1() -> (LOD_1:bool), Translation="0,0,0", Rotation="0,0,0", Scale="1,3,2") -> (Pose) {
    alphaInputType=Bool
    scaleMode=BMM_Additive
    scaleSpace=BCS_BoneSpace
    alphaBoolBlend {blendOption=Linear}
    boneToModify {boneName=neck_lod_2_corrected}
    scale {y=3, z=2}
   }
   "Transform (Modify) Bone - Bone: neck_lod_2_corrected"(ComponentPose=@n6, bAlphaBoolEnabled=Get LOD_2() -> (LOD_2:bool), Translation="0,0,0", Rotation="0,0,0", Scale="1,3,3") {
    alphaInputType=Bool
    scaleMode=BMM_Additive
    scaleSpace=BCS_BoneSpace
    alphaBoolBlend {blendOption=Linear}
    boneToModify {boneName=neck_lod_2_corrected}
    scale {y=3, z=3}
   }
   n7 = Layered blend per bone(
    BasePose=Use cached pose 'startpose'()
    BlendPoses_0=Slot 'DefaultSlot'()
    BlendWeights_0=Get Custom Idle is Playing() -> (Custom Idle is Playing:float)
   ) {layerSetup [{branchFilters [
       {boneName=dyn_BACE_chain_small}
       {boneName=dyn_chain_small_0}
       {boneName=dyn_chain_small_1}
       {boneName=dyn_sword_0}
       {boneName=dyn_chain_long_PARENT}
       {boneName=dyn_chain_long_3_KINEM_BACE}
       {boneName=dyn_chain_long_3_KINEM}
       {boneName=dyn_chain_long_1_KINEM}
       {boneName=dyn_chain_long_2_KINEM}
       {boneName=dyn_chain_long_end1_KINEM}
       {boneName=dyn_chain_long_end1_KINEM}
      ]}]}
   Output Pose(Result=@n7) {layerGroup=None}
   n8 = Control Rig!(Source=Copy Pose From Mesh() {bUseAttachedParent=true}, Alpha, LOD_1_CR:bool=Get LOD_1(), LOD_2_CR:bool=Get LOD_2()) -> (Pose) {
    alphaBoolBlend {blendOption=Linear}
    controlRigAssetReference {blueprintRigClass=&/F_MED_CeremonialGuard_Fencer__main_CtrlRig.F_MED_CeremonialGuard_Fencer__main_CtrlRig_C}
   }
   Save cached pose 'Pre sim'(Pose=@n8) {cacheName=Pre sim}
   n9 = Control Rig!(Source=Use cached pose 'Pre sim'(), Alpha, is sky_:bool=false) -> (Pose) {
    alphaBoolBlend {blendOption=Linear}
    controlRigAssetReference {blueprintRigClass=&/F_MED_CeremonialGuard_Fencer_CtrlRig.F_MED_CeremonialGuard_Fencer_CtrlRig_C}
   }
   n10 = Get Gravity Override()
   n11 = RigidBody!(ComponentPose=Local To Component(LocalPose=@n9), Alpha, OverrideWorldGravity:Vector=@n10, ExternalForce) -> (Pose) {
    bOverrideWorldGravity=true
    overridePhysicsAsset=&/F_MED_CeremonialGuard_Fencer_Physics
    simulationSpace=BaseBoneSpace
    alphaBoolBlend {blendOption=Linear}
    baseBoneRef {boneName=spine_05}
    componentAppliedLinearAccClamp=<30000 30000 30000>
    componentLinearAccScale {z=1}
    componentLinearVelScale {z=1}
    overrideWorldGravity {z=-400}
    simSpaceSettings {
     maxAngularAcceleration=200
     maxAngularVelocity=3
     maxLinearAcceleration=200
     maxLinearVelocity=150
     velocityScaleZ=0
     worldAlpha=1
     externalLinearDragV=<0.1 0.1>
    }
   }
   n12 = RigidBody(ComponentPose=@n11, OverrideWorldGravity=@n10) {
    bOverrideWorldGravity=true
    overridePhysicsAsset=&/F_MED_CeremonialGuard_Fencer_Physics_Coat_standart
    simulationSpace=BaseBoneSpace
    alphaBoolBlend {blendOption=Linear}
    baseBoneRef {boneName=pelvis}
    componentAppliedLinearAccClamp=<3000 3000 3000>
    componentLinearAccScale {z=1}
    componentLinearVelScale {z=1}
    overrideWorldGravity {z=-400}
    simSpaceSettings {
     maxAngularAcceleration=100
     maxAngularVelocity=3
     maxLinearAcceleration=600
     maxLinearVelocity=150
     velocityScaleZ=0
     worldAlpha=1
     externalLinearDragV=<0.1 0.1>
    }
   }
   n13 = RigidBody(ComponentPose=@n12, OverrideWorldGravity=@n10) {
    bClampLinearTranslationLimitToRefPose=true
    bOverrideWorldGravity=true
    overridePhysicsAsset=&/F_MED_CeremonialGuard_Fencer_Physics1
    simulationSpace=BaseBoneSpace
    alphaBoolBlend {blendOption=Linear}
    baseBoneRef {boneName=pelvis}
    componentAppliedLinearAccClamp=<3000 3000 3000>
    componentLinearAccScale {z=1}
    componentLinearVelScale {z=1}
    overrideWorldGravity {z=-400}
    simSpaceSettings {
     maxAngularAcceleration=100
     maxAngularVelocity=3
     maxLinearAcceleration=600
     maxLinearVelocity=150
     velocityScaleZ=0
     worldAlpha=1
     externalLinearDragV=<0.1 0.1>
    }
   }
   n14 = Control Rig!(Source=Use cached pose 'Pre sim'(), Alpha) -> (Pose) {
    alphaBoolBlend {blendOption=Linear}
    controlRigAssetReference {blueprintRigClass=&/F_MED_CeremonialGuard_Fencer__Crouch_CtrlRig.F_MED_CeremonialGuard_Fencer__Crouch_CtrlRig_C}
   }
   n15 = Get Gravity Override()
   n16 = RigidBody(ComponentPose=Local To Component(LocalPose=@n14), OverrideWorldGravity=@n15) {
    bOverrideWorldGravity=true
    overridePhysicsAsset=&F_MED_CeremonialGuard_Fencer_Physics
    simulationSpace=BaseBoneSpace
    alphaBoolBlend {blendOption=Linear}
    baseBoneRef {boneName=spine_05}
    componentAppliedLinearAccClamp=<30000 30000 30000>
    componentLinearAccScale {z=1}
    componentLinearVelScale {z=1}
    overrideWorldGravity {z=-400}
    simSpaceSettings {
     maxAngularAcceleration=200
     maxAngularVelocity=3
     maxLinearAcceleration=200
     maxLinearVelocity=150
     velocityScaleZ=0
     worldAlpha=1
     externalLinearDragV=<0.1 0.1>
    }
   }
   n17 = RigidBody(ComponentPose=@n16, OverrideWorldGravity=@n15) {
    bOverrideWorldGravity=true
    overridePhysicsAsset=&/F_MED_CeremonialGuard_Fencer_Physics_Coat
    simulationSpace=BaseBoneSpace
    alphaBoolBlend {blendOption=Linear}
    baseBoneRef {boneName=pelvis}
    componentAppliedLinearAccClamp=<3000 3000 3000>
    componentLinearAccScale {z=1}
    componentLinearVelScale {z=1}
    overrideWorldGravity {z=-400}
    simSpaceSettings {
     maxAngularAcceleration=100
     maxAngularVelocity=3
     maxLinearAcceleration=600
     maxLinearVelocity=150
     velocityScaleZ=0
     worldAlpha=1
     externalLinearDragV=<0.1 0.1>
    }
   }
   n18 = RigidBody(ComponentPose=@n17, OverrideWorldGravity=@n15) {
    bClampLinearTranslationLimitToRefPose=true
    bOverrideWorldGravity=true
    overridePhysicsAsset=&/F_MED_CeremonialGuard_Fencer_Physics_Crounh_Sword
    simulationSpace=BaseBoneSpace
    alphaBoolBlend {blendOption=Linear}
    baseBoneRef {boneName=pelvis}
    componentAppliedLinearAccClamp=<3000 3000 3000>
    componentLinearAccScale {z=1}
    componentLinearVelScale {z=1}
    overrideWorldGravity {z=-400}
    simSpaceSettings {
     maxAngularAcceleration=100
     maxAngularVelocity=3
     maxLinearAcceleration=600
     maxLinearVelocity=150
     velocityScaleZ=0
     worldAlpha=1
     externalLinearDragV=<0.1 0.1>
    }
   }
   n19 = "Transform (Modify) Bone - Bone: dyn_badge_down"(ComponentPose=@n18, Alpha=1, Translation="0.5,0,0", Rotation="0,0,0", Scale="1,1,1") {
    rotationMode=BMM_Additive
    translationMode=BMM_Additive
    translationSpace=BCS_BoneSpace
    alphaBoolBlend {blendOption=Linear}
    boneToModify {boneName=dyn_badge_down}
    translation {x=0.5}
   }
   n20 = Blend Poses by bool(
    BlendPose_0=Component To Local(ComponentPose=@n19)
    BlendPose_1=Component To Local(ComponentPose=@n13)
    BlendTime_0=0
    bActiveValue=Get bIsCrouching()
   ) {blendType=Linear}
   n21 = Control Rig(Source=Use cached pose 'Pre sim'()) {
    alphaBoolBlend {blendOption=Linear}
    controlRigAssetReference {blueprintRigClass=&F_MED_CeremonialGuard_Fencer__Crouch_CtrlRig_C}
   }
   n22 = Get Gravity Override()
   n23 = RigidBody(ComponentPose=Local To Component(LocalPose=@n21), OverrideWorldGravity=@n22) {
    bOverrideWorldGravity=true
    overridePhysicsAsset=&F_MED_CeremonialGuard_Fencer_Physics
    simulationSpace=BaseBoneSpace
    alphaBoolBlend {blendOption=Linear}
    baseBoneRef {boneName=spine_05}
    componentAppliedLinearAccClamp=<30000 30000 30000>
    componentLinearAccScale {z=1}
    componentLinearVelScale {z=1}
    overrideWorldGravity {z=-400}
    simSpaceSettings {
     maxAngularAcceleration=200
     maxAngularVelocity=3
     maxLinearAcceleration=200
     maxLinearVelocity=150
     velocityScaleZ=0
     worldAlpha=1
     externalLinearDragV=<0.1 0.1>
    }
   }
   n24 = RigidBody(ComponentPose=@n23, OverrideWorldGravity=@n22) {
    bOverrideWorldGravity=true
    overridePhysicsAsset=&F_MED_CeremonialGuard_Fencer_Physics_Coat
    simulationSpace=BaseBoneSpace
    alphaBoolBlend {blendOption=Linear}
    baseBoneRef {boneName=pelvis}
    componentAppliedLinearAccClamp=<3000 3000 3000>
    componentLinearAccScale {z=1}
    componentLinearVelScale {z=1}
    overrideWorldGravity {z=-400}
    simSpaceSettings {
     maxAngularAcceleration=100
     maxAngularVelocity=3
     maxLinearAcceleration=600
     maxLinearVelocity=150
     velocityScaleZ=0
     worldAlpha=1
     externalLinearDragV=<0.1 0.1>
    }
   }
   n25 = "Transform (Modify) Bone - Bone: dyn_chain_small_1"!(ComponentPose=@n24, bAlphaBoolEnabled:bool=True, Translation="4.676,-2.620,0.000", Rotation="-37.736,85.782,-88.292", Scale="1,1,1") -> (Pose) {
    alphaInputType=Bool
    rotationMode=BMM_Additive
    scaleSpace=BCS_BoneSpace
    translationMode=BMM_Additive
    alphaBoolBlend {blendOption=Linear}
    boneToModify {boneName=dyn_chain_small_1}
    rotation=<rot -37.736 85.782 -88.292>
    translation=<4.676 -2.62>
   }
   n26 = RigidBody(ComponentPose=@n25, OverrideWorldGravity=@n22) {
    bClampLinearTranslationLimitToRefPose=true
    bOverrideWorldGravity=true
    overridePhysicsAsset=&/F_MED_CeremonialGuard_Fencer_Physics_Sliding_Sword
    simulationSpace=BaseBoneSpace
    alphaBoolBlend {blendOption=Linear}
    baseBoneRef {boneName=pelvis}
    componentAppliedLinearAccClamp=<3000 3000 3000>
    componentLinearAccScale {z=1}
    componentLinearVelScale {z=1}
    overrideWorldGravity {z=-400}
    simSpaceSettings {
     maxAngularAcceleration=100
     maxAngularVelocity=3
     maxLinearAcceleration=600
     maxLinearVelocity=150
     velocityScaleZ=0
     worldAlpha=1
     externalLinearDragV=<0.1 0.1>
    }
   }
   n27 = "Transform (Modify) Bone - Bone: dyn_badge_down"(ComponentPose=@n26, Alpha=1, Translation="0.5,0,0", Rotation="0,0,0", Scale="1,1,1") {
    rotationMode=BMM_Additive
    translationMode=BMM_Additive
    translationSpace=BCS_BoneSpace
    alphaBoolBlend {blendOption=Linear}
    boneToModify {boneName=dyn_badge_down}
    translation {x=0.5}
   }
   n28 = Component To Local(ComponentPose=@n27)
   n29 = Look At - Bone: dyn_chain_long_1_KINEM(
    ComponentPose=Local To Component(LocalPose=Blend Poses by bool(BlendPose_0=@n28, BlendPose_1=@n20, BlendTime_0=0) {blendType=Linear})
    Alpha=1
   ) {
    alphaBoolBlend {blendOption=Linear}
    boneToModify {boneName=dyn_chain_long_1_KINEM}
    lookAtLocation=<0 100>
    lookAtTarget {boneReference {boneName=dyn_chain_long_PARENT}}
   }
   n30 = Look At - Bone: dyn_chain_long_PARENT(ComponentPose=@n29, Alpha=1) {
    alphaBoolBlend {blendOption=Linear}
    boneToModify {boneName=dyn_chain_long_PARENT}
    lookAtLocation=<0 100>
    lookAtTarget {boneReference {boneName=dyn_chain_long_1_KINEM}}
   }
   n31 = "Constraint - dyn_hand_ARM_R (PPPPPP)"!(
    ComponentPose=@n30
    Alpha=1
    ConstraintWeights_0=0.358
    ConstraintWeights_1:float=0.0541
    ConstraintWeights_2:float=0.0599
    ConstraintWeights_3:float=0.4026
    ConstraintWeights_4:float=0.1117
    ConstraintWeights_5:float=0.0137
   ) -> (Pose) {
    alphaBoolBlend {blendOption=Linear}
    boneToModify {boneName=dyn_hand_ARM_R}
    constraintSetup [
     {targetBone {boneName=hand_l}}
     {targetBone {boneName=thumb_01_l}}
     {targetBone {boneName=index_metacarpal_l}}
     {targetBone {boneName=middle_metacarpal_l}}
     {targetBone {boneName=ring_metacarpal_l}}
     {targetBone {boneName=pinky_metacarpal_l}}
    ]
    constraintWeights [1,1,1,1,1,1]
   }
   n32 = Anim Dynamics - Bone: dyn_BACE_tongue_down(ComponentPose=@n31, Alpha=1, ExternalForce="0,0,0") {
    angularDampingOverride=0.9999
    angularSpringConstant=250
    bAngularSpring=true
    bOverrideAngularDamping=true
    bUseGravityOverride=true
    simulationSpace=RootRelative
    alphaBoolBlend {blendOption=Linear}
    boundBone {boneName=dyn_BACE_tongue_down}
    componentAppliedLinearAccClamp=<1000 1000 1010>
    componentLinearAccScale=<1 1 1>
    componentLinearVelScale=<1 1 1>
    gravityOverride {z=-350}
    physicsBodyDefinitions [{
      boundBone {boneName=dyn_BACE_tongue_down}
      boxExtents=<2 2 2>
      constraintSetup {angularLimitsMax=<0 5 20>, angularLimitsMin=<0 -5 -30>, angularTarget=<0 0 5>}
      localJointOffset=<0.1 0 0>
     }]
    relativeSpaceBone {boneName=calf_twist_02_r}
   }
   n33 = Anim Dynamics - Bone: dyn_Belt_BACEtongue_up1(ComponentPose=@n32, Alpha=Get NotBackbling() -> (NotBackbling:double), ExternalForce="0,0,0") {
    angularDampingOverride=0.9999
    angularSpringConstant=250
    bAngularSpring=true
    bOverrideAngularDamping=true
    bUseGravityOverride=true
    simulationSpace=RootRelative
    alphaBoolBlend {blendOption=Linear}
    boundBone {boneName=dyn_Belt_BACEtongue_up1}
    componentAppliedLinearAccClamp=<1000 1000 1010>
    componentLinearAccScale=<1 1 1>
    componentLinearVelScale=<1 1 1>
    gravityOverride {z=-350}
    physicsBodyDefinitions [{
      boundBone {boneName=dyn_Belt_BACEtongue_up1}
      boxExtents=<2 2 2>
      constraintSetup {angularLimitsMax=<0 5 20>, angularLimitsMin=<0 -5 -30>, angularTarget=<0 0 5>}
      localJointOffset=<0.1 0 0>
     }]
    relativeSpaceBone {boneName=calf_twist_02_r}
   }
   n34 = "Transform (Modify) Bone - Bone: dyn_Belt_BACEtongue_up1"(
    ComponentPose=@n33
    Alpha=Get IsShortBackblingOnAlpha() -> (IsShortBackblingOnAlpha:double)
    Translation="0,0,0"
    Rotation="-0.005,70.000,0.004"
    Scale="1,1,1"
   ) {
    rotationMode=BMM_Additive
    rotationSpace=BCS_BoneSpace
    alphaBoolBlend {blendOption=Linear}
    boneToModify {boneName=dyn_Belt_BACEtongue_up1}
    rotation=<rot -0.005 70 0.004>
   }
   n35 = Control Rig(Source=Component To Local(ComponentPose=@n34), Alpha=0.6) {
    alphaBoolBlend {blendOption=Linear}
    controlRigAssetReference {blueprintRigClass=&/F_MED_CeremonialGuard_Fencer_Deforms_ControlRig.F_MED_CeremonialGuard_Fencer_Deforms_ControlRig_C}
   }
   Save cached pose 'startpose'(Pose=@n35) {cacheName=startpose}
   Sequence Player() {bLoopAnimation=false}
  ]}
 EventGraph {nodes [
   Event BlueprintUpdateAnimation()
   > n1 = Sequence!(execute) -> (then_0, then_1, then_2:Exec)
   > n2 = Gravity Override State Function DEPRECATED(CharacterPartAnimInstance=Self-Reference(), GameTimeInSeconds=GetGameTimeInSeconds(), GravityOverrideParams=Get Gravity Override Params())
   > Set Gravity Override(Gravity Override=@n2)
   > Set sky_AB(sky_AB:bool=Get bIsSkyDiving()) -> (Output_Get:bool)
   n3 = Get CurrentLOD()
   > Set LOD_1(LOD_1:bool="Equal (Integer)"(A=@n3, B=1)) -> (Output_Get:bool)
   > Set LOD_2(LOD_2:bool="integer >= integer"(A=@n3, B=2)) -> (Output_Get:bool)
   n4 = CheckForBackbling()
   n5 = Set IsWearingBackbling(execute=@n1.then_1, IsWearingBackbling:bool=@n4) -> (Output_Get:bool)
   > Set Is Backbling A Cape(Is Backbling A Cape:bool=@n4.bIsBackblingACape) -> (Output_Get:bool)
   > Set NotBackbling(NotBackbling:double="To Float (Boolean)"(InBool=NOT Boolean(A=@n4))) -> (Output_Get:double)
   > Set is Wear a Cape(is Wear a Cape:bool=AND Boolean(A=@n4, B=@n4.bIsBackblingLong)) -> (Output_Get:bool)
   > n6 = Branch(Condition=@n5)
   > n7 = Branch(Condition=@n4.bIsBackblingLong)
   Set IsShortBackblingOnAlpha(execute=[@n7.then, @n6.else], IsShortBackblingOnAlpha:double=0) -> (Output_Get:double)
   Set IsShortBackblingOnAlpha(execute=@n7.else, IsShortBackblingOnAlpha=1)
   Branch(execute=@n1.then_2, Condition=Get bIsInFrontEnd())
   > Set Custom Idle is Playing(Custom Idle is Playing:float=GetCurveValue(self=GetOwnerFrontEndAnimInstance(), CurveName=Frontend_Idle_Bypass)) -> (Output_Get:float)
   n8 = Event OnCharacterCustomizationLoaded()
   n9 = Get Dynamic Material Instance(execute=[@n8.then, @n11.then, @n12.then], SkeletalMeshComponent=GetPartSkeletalMeshComponent())
   > n10 = Branch(Condition=PickaxeItemDefinitionHasAnyMetaTag(Tags=Make Array("[0]"=(TagName="Cosmetics.MetaTags.CeremonialGuard")), bRequiresEquipped=true))
   > SetScalarParameterValue(self=@n9, ParameterName=Hide Element 02, Value=1)
   SetScalarParameterValue(execute=@n10.else, self=@n9, ParameterName=Hide Element 02, Value=0)
   n11 = Event OnEndPlayingEmote() {is=title K2Node_Event}
   n12 = Event OnWeaponChanged()
  ]}
]
variables [
 Custom Idle is Playing {type=number}
 Gravity Override {type=Vector}
 Gravity Override Params {type=GravityOverrideParamsStruct}
 Is Backbling A Cape {type=boolean}
 is Wear a Cape {type=boolean}
 IsShortBackblingOnAlpha {type=number}
 IsWearingBackbling {type=boolean}
 LOD_1 {type=boolean}
 LOD_2 {type=boolean}
 NotBackbling {type=number}
 sky_AB {type=boolean}
]
